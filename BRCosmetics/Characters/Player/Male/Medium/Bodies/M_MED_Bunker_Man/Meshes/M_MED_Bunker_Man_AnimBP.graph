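# Post-process style AnimBlueprint for the Bunker Man male body cosmetic, adding secondary motion on top of a pose copied from its attached parent mesh.
# Spring-driven AnimDynamics nodes sway three armband bones, then constraints tie foot and ball offset bones to their source bones.
# Additive bone-space translations shift front and back scarf bones, weighted by head_down_alpha and head_up_alpha, which each update derives from head-to-spine_01 delta yaw from the reference pose.
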AnimBlueprint /BRCosmetics/Characters/Player/Male/Medium/Bodies/M_MED_Bunker_Man/Meshes/M_MED_Bunker_Man_AnimBP
target_skeleton=&/M_MED_Bunker_Man_Skeleton
graphs [
 AnimGraph {nodes [
   n1 = Anim Dynamics - Bone: dyn_armband_knot(ComponentPose=Local To Component(LocalPose=Copy Pose From Mesh() {bUseAttachedParent=true}), Alpha=0.75, ExternalForce="0,0,0") {
    alpha=0.9
    angularDampingOverride=0.98
    angularSpringConstant=1500
    bAngularSpring=true
    bGravityOverrideInSimSpace=true
    bOverrideAngularDamping=true
    numSolverIterationsPostUpdate=4
    numSolverIterationsPreUpdate=16
    simulationSpace=Actor
    alphaBoolBlend {blendOption=Linear}
    boundBone {boneName=dyn_armband_knot}
    physicsBodyDefinitions [{
      sphereCollisionRadius=0
      boundBone {boneName=dyn_armband_knot}
      boxExtents=<15 5 5>
      constraintSetup {angularTargetAxis=AxisZ, twistAxis=AxisY, angularLimitsMax=<15 0 15>, angularLimitsMin=<-15 0 -15>, angularTarget=<0 0 1>}
      localJointOffset=<0 0 0>
     }]
   }
   n2 = Anim Dynamics - Bone: dyn_armband_low(ComponentPose=@n1, Alpha=0.5, ExternalForce="0,0,0") {
    alpha=0.9
    angularDampingOverride=0.98
    angularSpringConstant=1500
    bAngularSpring=true
    bGravityOverrideInSimSpace=true
    bOverrideAngularDamping=true
    numSolverIterationsPostUpdate=4
    numSolverIterationsPreUpdate=16
    simulationSpace=Actor
    alphaBoolBlend {blendOption=Linear}
    boundBone {boneName=dyn_armband_low}
    physicsBodyDefinitions [{
      sphereCollisionRadius=0
      boundBone {boneName=dyn_armband_low}
      boxExtents=<15 5 5>
      constraintSetup {angularTargetAxis=AxisZ, twistAxis=AxisY, angularLimitsMax=<15 15 15>, angularLimitsMin=<-15 -15 -15>, angularTarget=<0 0 1>}
      localJointOffset=<0 0 0>
     }]
   }
   n3 = Anim Dynamics - Bone: dyn_armband_hi(ComponentPose=@n2, Alpha=0.5, ExternalForce="0,0,0") {
    alpha=0.9
    angularDampingOverride=0.98
    angularSpringConstant=1500
    bAngularSpring=true
    bGravityOverrideInSimSpace=true
    bOverrideAngularDamping=true
    numSolverIterationsPostUpdate=4
    numSolverIterationsPreUpdate=16
    simulationSpace=Actor
    alphaBoolBlend {blendOption=Linear}
    boundBone {boneName=dyn_armband_hi}
    physicsBodyDefinitions [{
      sphereCollisionRadius=0
      boundBone {boneName=dyn_armband_hi}
      boxExtents=<15 5 5>
      constraintSetup {angularTargetAxis=AxisZ, twistAxis=AxisY, angularLimitsMax=<15 15 15>, angularLimitsMin=<-15 -15 -15>, angularTarget=<0 0 1>}
      localJointOffset=<0 0 0>
     }]
   }
   n4 = "Constraint - foot_offset_l (P)"(ComponentPose=@n3, Alpha=1, ConstraintWeights_0=1) {
    alphaBoolBlend {blendOption=Linear}
    boneToModify {boneName=foot_offset_l}
    constraintSetup [{targetBone {boneName=foot_l}}]
    constraintWeights [1]
   }
   n5 = "Constraint - foot_offset_r (P)"(ComponentPose=@n4, Alpha=1, ConstraintWeights_0=1) {
    alphaBoolBlend {blendOption=Linear}
    boneToModify {boneName=foot_offset_r}
    constraintSetup [{targetBone {boneName=foot_r}}]
    constraintWeights [1]
   }
   n6 = "Constraint - ball_offset_l (P)"(ComponentPose=@n5, Alpha=1, ConstraintWeights_0=1) {
    alphaBoolBlend {blendOption=Linear}
    boneToModify {boneName=ball_offset_l}
    constraintSetup [{targetBone {boneName=ball_l}}]
    constraintWeights [1]
   }
   n7 = "Constraint - ball_offset_r (P)"(ComponentPose=@n6, Alpha=1, ConstraintWeights_0=1) {
    alphaBoolBlend {blendOption=Linear}
    boneToModify {boneName=ball_offset_r}
    constraintSetup [{targetBone {boneName=ball_r}}]
    constraintWeights [1]
   }
   n8 = "Transform (Modify) Bone - Bone: dyn_scarf_f"(ComponentPose=@n7, Alpha=Get head_down_alpha() -> (head_down_alpha:double), Translation="0,-10,0", Rotation="0,0,0", Scale="1,1,1") {
    translationMode=BMM_Additive
    translationSpace=BCS_BoneSpace
    alphaBoolBlend {blendOption=Linear}
    boneToModify {boneName=dyn_scarf_f}
    translation {y=-10}
   }
   n9 = "Transform (Modify) Bone - Bone: dyn_scarf_b"(ComponentPose=@n8, Alpha=Get head_up_alpha() -> (head_up_alpha:double), Translation="0,10,0", Rotation="0,0,0", Scale="1,1,1") {
    translationMode=BMM_Additive
    translationSpace=BCS_BoneSpace
    alphaBoolBlend {blendOption=Linear}
    boneToModify {boneName=dyn_scarf_b}
    translation {y=10}
   }
   Output Pose(Result=Component To Local(ComponentPose=@n9)) {layerGroup=None}
  ]}
 EventGraph {nodes [
   Event BlueprintUpdateAnimation()
   > Sequence()
   n1 = GetOwnerSkeletalMeshComponent()
   n2 = BreakRotator(InRot=BreakTransform(InTransform=GetDeltaTransformFromRefPose(self=@n1, BoneName=spine_01, BaseName=head)).Rotation)
   > Set head_down_alpha(head_down_alpha:double=float - float(B=MapRangeClamped(Value=@n2.Yaw, InRangeA=-70, InRangeB=70)) {is=title K2Node_CallFunction}) -> (Output_Get:double)
   > Set head_up_alpha(head_up_alpha:double=MapRangeClamped(Value=@n2.Yaw, InRangeA=-70, InRangeB=70)) -> (Output_Get:double)
  ]}
]
variables [
 head_down_alpha {type=number}
 head_up_alpha {type=number}
]
