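# Physics asset for the beard part of the male medium Bunker Man cosmetic body, defining bodies for a long dynamic beard chain.
# Capsules on the upper spine, pelvis, thighs and calves approximate the character's body alongside the beard's boxes.
# The chain's root links from spine_03 use tiny boxes and free angular constraints down to dyn_beard_04.
# From there to dyn_beard_10, larger boxes have tight twist limits and swing limits that widen toward the tip.
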PhysicsAsset /BRCosmetics/Characters/Player/Male/Medium/Bodies/M_MED_Bunker_Man/Meshes/Parts/M_MED_BunkerMan_Beard
bodies {
 calf_l [{shapeType=Capsule, length=31.242, radius=11.9116, center=<-19.4843 -0.5441 -0.2294>, rotation=<rot 0 94.0793 -89.4659>}]
 calf_r [{shapeType=Capsule, length=31.5362, radius=11.5971, center=<19.4486 0.3664 0.2406>, rotation=<rot 0 93.6446 -89.3731>}]
 dyn_beard_01 [{shapeType=Box, extentX=1.01, extentY=1.01, extentZ=1.01, rotation=<rot 0 -90 0>}]
 dyn_beard_02 [{shapeType=Box, extentX=1.01, extentY=1.01, extentZ=1.01, rotation=<rot 0 -90 0>}]
 dyn_beard_03 [{shapeType=Box, extentX=1.01, extentY=1.01, extentZ=1.01, rotation=<rot 0 -90 0>}]
 dyn_beard_04 [{shapeType=Box, extentX=22.7264, extentY=18.5084, extentZ=11.2594, center=<7.7995 0.6178 0.4455>}]
 dyn_beard_05 [{shapeType=Box, extentX=15.3546, extentY=9.7949, extentZ=16.0811, center=<5.724 0.2904 0.6256>, rotation=<rot 0 177.7371 -88.3215>}]
 dyn_beard_06 [{shapeType=Box, extentX=15, extentY=12.5, extentZ=5, center=<1 0 0>}]
 dyn_beard_07 [{shapeType=Box, extentX=13.5, extentY=11, extentZ=5, center=<1 0 0>}]
 dyn_beard_08 [{shapeType=Box, extentX=13, extentY=11, extentZ=5}]
 dyn_beard_09 [{shapeType=Box, extentX=18.5, extentY=11, extentZ=5, center=<2.3435 0 -0.0001>}]
 dyn_beard_10 [{shapeType=Box, extentX=3.3685, extentY=3.1591, extentZ=13.2049, center=<4.0267 -0.301 0.4295>, rotation=<rot 0 90.5217 -84.0994>}]
 parent_dyn_beard_02 [{shapeType=Box, extentX=1.01, extentY=1.01, extentZ=1.01, rotation=<rot 0 -90 0>}]
 pelvis [{shapeType=Capsule, length=11.9559, radius=13.8329, center=<2.2747 -0.2766 1.0154>}]
 spine_03 [{shapeType=Capsule, length=15.0169, radius=12.0072, center=<-1.7779 1.5556 0.0528>, rotation=<rot 0.0008 6.275 6.276>}]
 thigh_l [{shapeType=Capsule, length=26.002, radius=10.5818, center=<-19.839 -0.0002 0>, rotation=<rot 90 0 0>}]
 thigh_r [{shapeType=Capsule, length=25.6153, radius=10.4315, center=<19.0128 0 0>, rotation=<rot 90 0 0>}]
}
constraints [
 {
  bone1Name=spine_03
  bone2Name=pelvis
  swing1LimitDegrees=45
  swing1Motion=Free
  swing2LimitDegrees=45
  swing2Motion=Free
  twistLimitDegrees=45
  twistMotion=Free
 }
 {
  bone1Name=thigh_l
  bone2Name=pelvis
  swing1LimitDegrees=45
  swing1Motion=Free
  swing2LimitDegrees=45
  swing2Motion=Free
  twistLimitDegrees=45
  twistMotion=Free
 }
 {
  bone1Name=calf_l
  bone2Name=thigh_l
  swing1LimitDegrees=45
  swing1Motion=Free
  swing2LimitDegrees=45
  swing2Motion=Free
  twistLimitDegrees=45
  twistMotion=Free
 }
 {
  bone1Name=thigh_r
  bone2Name=pelvis
  swing1LimitDegrees=45
  swing1Motion=Free
  swing2LimitDegrees=45
  swing2Motion=Free
  twistLimitDegrees=45
  twistMotion=Free
 }
 {
  bone1Name=calf_r
  bone2Name=thigh_r
  swing1LimitDegrees=45
  swing1Motion=Free
  swing2LimitDegrees=45
  swing2Motion=Free
  twistLimitDegrees=45
  twistMotion=Free
 }
 {bone1Name=dyn_beard_06, bone2Name=dyn_beard_05, swing1LimitDegrees=35, swing2LimitDegrees=35, twistLimitDegrees=5}
 {bone1Name=dyn_beard_07, bone2Name=dyn_beard_06, swing1LimitDegrees=35, swing2LimitDegrees=35, twistLimitDegrees=5}
 {bone1Name=dyn_beard_08, bone2Name=dyn_beard_07, swing1LimitDegrees=35, swing2LimitDegrees=35, twistLimitDegrees=10}
 {bone1Name=dyn_beard_09, bone2Name=dyn_beard_08, swing1LimitDegrees=40, swing2LimitDegrees=40, twistLimitDegrees=10}
 {bone1Name=dyn_beard_10, bone2Name=dyn_beard_09, swing1LimitDegrees=45, swing2LimitDegrees=45, twistLimitDegrees=10}
 {bone1Name=dyn_beard_05, bone2Name=dyn_beard_04, swing1LimitDegrees=35, swing2LimitDegrees=35, twistLimitDegrees=10}
 {
  bone1Name=dyn_beard_04
  bone2Name=dyn_beard_03
  swing1LimitDegrees=45
  swing1Motion=Free
  swing2LimitDegrees=45
  swing2Motion=Free
  twistLimitDegrees=45
  twistMotion=Free
 }
 {
  bone1Name=dyn_beard_03
  bone2Name=dyn_beard_02
  swing1LimitDegrees=45
  swing1Motion=Free
  swing2LimitDegrees=45
  swing2Motion=Free
  twistLimitDegrees=45
  twistMotion=Free
 }
 {
  bone1Name=dyn_beard_02
  bone2Name=parent_dyn_beard_02
  swing1LimitDegrees=45
  swing1Motion=Free
  swing2LimitDegrees=45
  swing2Motion=Free
  twistLimitDegrees=45
  twistMotion=Free
 }
 {
  bone1Name=parent_dyn_beard_02
  bone2Name=dyn_beard_01
  swing1LimitDegrees=45
  swing1Motion=Free
  swing2LimitDegrees=45
  swing2Motion=Free
  twistLimitDegrees=45
  twistMotion=Free
 }
 {
  bone1Name=dyn_beard_01
  bone2Name=spine_03
  swing1LimitDegrees=45
  swing1Motion=Free
  swing2LimitDegrees=45
  swing2Motion=Free
  twistLimitDegrees=45
  twistMotion=Free
 }
]
