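# Control Rig for the Bunker Man character's beard that re-aims the root beard bone, dyn_beard_01, toward a point offset from the chest.
# It transforms a fixed local Y offset by spine_04, then applies a from-two-vectors quaternion to swing the beard bone's X axis toward that point.
# The global rotation propagates to child beard, braid and moustache bones.
# Debug axes draw the bone before and after rotation, plus the target point.
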ControlRigBlueprint /BRCosmetics/Characters/Player/Male/Medium/Bodies/M_MED_Bunker_Man/Meshes/Parts/M_MED_BunkerMan_Beard_CtrlRig
graphs [
 RigVMFunctionLibrary {links [], nodes []}
 RigVMModel {nodes [
   n1 = &RigUnit_BeginExecution()
   /Script/ControlRig.RigUnit_DebugTransformMutable(
    Color=(R=1,G=0,B=0,A=1)
    DebugDrawSettings=(DepthPriority=SDPG_Foreground,Lifetime=-1)
    Mode=Axes
    Scale=50
    Space=dyn_beard_01
    Thickness=0
    Transform=(Rotation=(X=0,Y=0,Z=0,W=1),Translation=(X=0,Y=0,Z=0),Scale3D=(X=1,Y=1,Z=1))
    WorldOffset=(Rotation=(X=0,Y=0,Z=0,W=1),Translation=(X=0,Y=0,Z=0),Scale3D=(X=1,Y=1,Z=1))
    bEnabled=True
    ExecutePin=@n1.ExecutePin
   ) {is=struct RigVMUnitNode}
   n2 = &RigUnit_GetBoneTransform(Bone=dyn_beard_01)
   n3 = /Script/RigVM.RigVMFunction_MathQuaternionGetAxis(Axis=X, Quaternion=@n2.Transform.Rotation) {is=struct RigVMUnitNode, template="Axis::Execute(in Quaternion,in Axis,out Result)"}
   n4 = /Script/RigVM.RigVMFunction_MathTransformTransformVector(
    Transform=&RigUnit_GetBoneTransform(Bone=spine_04).Transform
    Location=/Script/RigVM.RigVMFunction_MathVectorAbs(Value=(X=0,Y=20,Z=0)) {is=struct RigVMUnitNode, template="Absolute::Execute(in Value,out Result)"}.Result
   ) {is=struct RigVMUnitNode}
   n5 = /Script/RigVM.RigVMFunction_MathVectorSub(B=@n2.Transform.Translation, A=@n4.Result) {is=struct RigVMUnitNode, template="Subtract::Execute(in A,in B,out Result)"}
   > /Script/ControlRig.RigUnit_SetBoneRotation(
    Bone=dyn_beard_01
    Space=GlobalSpace
    Weight=1
    bPropagateToChildren=True
    Rotation=/Script/RigVM.RigVMFunction_MathQuaternionMul(
     B=@n2.Transform.Rotation
     A=/Script/RigVM.RigVMFunction_MathQuaternionFromTwoVectors(A=@n3.Result, B=@n5.Result) {is=struct RigVMUnitNode, template="FromTwoVectors::Execute(in A,in B,out Result)"}.Result
    ) {is=struct RigVMUnitNode, template="Multiply::Execute(in A,in B,out Result)"}.Result
   ) {is=struct RigVMUnitNode}
   > &RigUnit_DebugTransformMutable(
    Color=(R=1,G=0,B=0,A=1)
    DebugDrawSettings=(DepthPriority=SDPG_Foreground,Lifetime=-1)
    Mode=Axes
    Scale=75
    Space=dyn_beard_01
    Thickness=0
    Transform=(Rotation=(X=0,Y=0,Z=0,W=1),Translation=(X=0,Y=0,Z=0),Scale3D=(X=1,Y=1,Z=1))
    WorldOffset=(Rotation=(X=0,Y=0,Z=0,W=1),Translation=(X=0,Y=0,Z=0),Scale3D=(X=1,Y=1,Z=1))
    bEnabled=True
   )
   > &RigUnit_DebugTransformMutable(
    Color=(R=1,G=0,B=0,A=1)
    DebugDrawSettings=(DepthPriority=SDPG_Foreground,Lifetime=-1)
    Mode=Axes
    Scale=55
    Space=None
    Thickness=0
    Transform=(Rotation=(X=0,Y=0,Z=0,W=1),Translation=(X=0,Y=0,Z=0),Scale3D=(X=1,Y=1,Z=1))
    WorldOffset=(Rotation=(X=0,Y=0,Z=0,W=1),Translation=(X=0,Y=0,Z=0),Scale3D=(X=1,Y=1,Z=1))
    bEnabled=True
    Transform.Translation=@n4.Result
   )
  ]}
]
hierarchy [
 root {loc=<0 0 0>, rot=<0 0 0 1>}
 pelvis {parent=root, loc=<0 -0.7071 93.1191>, rot=<0.0127 0.707 -0.0127 -0.707>}
 spine_01 {parent=pelvis, loc=<6.7494 -0.6388 0>, rot=<0 0 0.034 -0.9994>}
 spine_02 {parent=spine_01, loc=<7.2518 -0.035 0>, rot=<0 0 0.0238 -0.9997>}
 spine_03 {parent=spine_02, loc=<6.0104 0 0>, rot=<0 0 0.0149 -0.9999>}
 spine_04 {parent=spine_03, loc=<8.1024 0 0>, rot=<0 0 -0.0132 -0.9999>}
 spine_05 {parent=spine_04, loc=<10.0146 0 0>, rot=<0 0 -0.0015 -1>}
 clavicle_l {parent=spine_05, loc=<4.3575 -0.1466 -2.3223>, rot=<0.0465 -0.7289 -0.009 -0.6829>}
 upperarm_l {parent=clavicle_l, loc=<15.5111 0 0>, rot=<0.0226 -0.3234 -0.0088 -0.9459>}
 lowerarm_l {parent=upperarm_l, loc=<26.6514 0 0>, rot=<0.0391 -0.0099 -0.2453 -0.9686>}
 hand_l {parent=lowerarm_l, loc=<22.3491 0 0>, rot=<-0.7081 -0.0507 0.0975 -0.6975>}
 index_metacarpal_l {parent=hand_l, loc=<1.6535 0.0183 -2.96>, rot=<0.0158 -0.0124 -0.0173 -0.9996>}
 index_01_l {parent=index_metacarpal_l, loc=<8.3008 0.0357 -0.0524>, rot=<-0.0044 0.0485 0.0924 -0.9945>}
 index_02_l {parent=index_01_l, loc=<3.3587 -0.3004 0.0452>, rot=<-0.002 0.0027 0.3225 -0.9466>}
 index_03_l {parent=index_02_l, loc=<2.5578 -0.123 -0.0252>, rot=<-0.002 0.0067 0.2558 -0.9667>}
 middle_metacarpal_l {parent=hand_l, loc=<2.1856 -0.1725 -0.5984>, rot=<0.006 0.0168 -0.0329 -0.9993>}
 middle_01_l {parent=middle_metacarpal_l, loc=<7.3275 -0.2447 0.0129>, rot=<-0.0208 0.0042 0.0858 -0.9961>}
 middle_02_l {parent=middle_01_l, loc=<4.0123 0.1401 0.1175>, rot=<-0.0218 0.0258 0.3512 -0.9357>}
 middle_03_l {parent=middle_02_l, loc=<2.6077 0.0534 0.1373>, rot=<0.0082 0.0188 0.1708 -0.9851>}
 pinky_metacarpal_l {parent=hand_l, loc=<1.7616 -0.9753 3.3814>, rot=<-0.0397 0.0814 -0.069 -0.9935>}
 pinky_01_l {parent=pinky_metacarpal_l, loc=<7.1709 -0.4831 -0.1094>, rot=<-0.0654 -0.0035 0.2242 -0.9723>}
 pinky_02_l {parent=pinky_01_l, loc=<2.8274 0.0893 -0.1043>, rot=<0.0037 0.0011 0.2854 -0.9584>}
 pinky_03_l {parent=pinky_02_l, loc=<2.2046 0.0307 -0.0078>, rot=<-0.0002 0.0098 0.1445 -0.9895>}
 ring_metacarpal_l {parent=hand_l, loc=<2.2074 -0.3625 1.3638>, rot=<-0.0284 0.0722 -0.0458 -0.9959>}
 ring_01_l {parent=ring_metacarpal_l, loc=<7.0295 -0.6197 -0.1446>, rot=<-0.0489 -0.0252 0.1892 -0.9804>}
 ring_02_l {parent=ring_01_l, loc=<3.6174 0.4646 -0.1027>, rot=<0 0 0.2806 -0.9598>}
 ring_03_l {parent=ring_02_l, loc=<2.4152 0.1141 0.0199>, rot=<0 0 0.1189 -0.9929>}
 thumb_01_l {parent=hand_l, loc=<1.3118 -0.3191 -2.459>, rot=<0.6546 -0.4042 0.0664 -0.6354>}
 thumb_02_l {parent=thumb_01_l, loc=<5.9336 -0.0448 0.0401>, rot=<-0.0037 0.0213 0.298 -0.9543>}
 thumb_03_l {parent=thumb_02_l, loc=<2.2223 0.3682 -0.0907>, rot=<0.0118 -0.013 0.0066 -0.9998>}
 lowerarm_twist_01_l {parent=lowerarm_l, loc=<17.8549 0 0>, rot=<0 0 0 1>}
 lowerarm_twist_02_l {parent=lowerarm_l, loc=<9.6199 0 0>, rot=<0 0 0 1>}
 upperarm_twist_01_l {parent=upperarm_l, loc=<0.0041 0 0>, rot=<0 0 0 -1>}
 upperarm_twist_02_l {parent=upperarm_l, loc=<13.7941 0 0>, rot=<0 0 0 -1>}
 clavicle_r {parent=spine_05, loc=<4.3575 -0.1466 2.3223>, rot=<0.7289 0.0465 -0.6829 0.009>}
 upperarm_r {parent=clavicle_r, loc=<-15.5111 0 0>, rot=<0.0226 -0.3234 -0.0088 -0.9459>}
 lowerarm_r {parent=upperarm_r, loc=<-26.6514 0 0>, rot=<0.0391 -0.0099 -0.2453 -0.9686>}
 hand_r {parent=lowerarm_r, loc=<-22.3492 0 0>, rot=<-0.7081 -0.0507 0.0975 -0.6975>}
 index_metacarpal_r {parent=hand_r, loc=<-1.6534 -0.0181 2.9601>, rot=<0.0158 -0.0124 -0.0173 -0.9996>}
 index_01_r {parent=index_metacarpal_r, loc=<-8.3008 -0.0357 0.0524>, rot=<-0.0044 0.0485 0.0924 -0.9945>}
 index_02_r {parent=index_01_r, loc=<-3.3589 0.3003 -0.0453>, rot=<-0.002 0.0027 0.3225 -0.9466>}
 index_03_r {parent=index_02_r, loc=<-2.5578 0.1231 0.0251>, rot=<-0.002 0.0067 0.2558 -0.9667>}
 middle_metacarpal_r {parent=hand_r, loc=<-2.1856 0.1726 0.5984>, rot=<0.006 0.0168 -0.0329 -0.9993>}
 middle_01_r {parent=middle_metacarpal_r, loc=<-7.3273 0.2447 -0.0129>, rot=<-0.0208 0.0042 0.0858 -0.9961>}
 middle_02_r {parent=middle_01_r, loc=<-4.0142 -0.1411 -0.07>, rot=<-0.0217 0.0259 0.3512 -0.9357>}
 middle_03_r {parent=middle_02_r, loc=<-2.6074 -0.0534 -0.1312>, rot=<0.0083 0.0188 0.1708 -0.9851>}
 pinky_metacarpal_r {parent=hand_r, loc=<-1.7617 0.9753 -3.3814>, rot=<-0.0397 0.0814 -0.069 -0.9935>}
 pinky_01_r {parent=pinky_metacarpal_r, loc=<-7.1707 0.4832 0.1094>, rot=<-0.0654 -0.0035 0.2242 -0.9723>}
 pinky_02_r {parent=pinky_01_r, loc=<-2.8274 -0.0892 0.1043>, rot=<0.0037 0.0011 0.2854 -0.9584>}
 pinky_03_r {parent=pinky_02_r, loc=<-2.2045 -0.031 0.0078>, rot=<-0.0002 0.0098 0.1445 -0.9895>}
 ring_metacarpal_r {parent=hand_r, loc=<-2.2073 0.3627 -1.3638>, rot=<-0.0284 0.0722 -0.0458 -0.9959>}
 ring_01_r {parent=ring_metacarpal_r, loc=<-7.0297 0.6197 0.1446>, rot=<-0.0489 -0.0252 0.1892 -0.9804>}
 ring_02_r {parent=ring_01_r, loc=<-3.6172 -0.4648 0.1026>, rot=<0 0 0.2806 -0.9598>}
 ring_03_r {parent=ring_02_r, loc=<-2.4152 -0.1141 -0.0199>, rot=<0 0 0.1189 -0.9929>}
 thumb_01_r {parent=hand_r, loc=<-1.3119 0.3192 2.459>, rot=<0.6546 -0.4042 0.0664 -0.6354>}
 thumb_02_r {parent=thumb_01_r, loc=<-5.9335 0.0447 -0.0401>, rot=<-0.0037 0.0213 0.298 -0.9543>}
 thumb_03_r {parent=thumb_02_r, loc=<-2.2223 -0.3682 0.0907>, rot=<0.0118 -0.013 0.0066 -0.9998>}
 lowerarm_twist_01_r {parent=lowerarm_r, loc=<-17.855 0 0>, rot=<0 0 0 1>}
 lowerarm_twist_02_r {parent=lowerarm_r, loc=<-9.6199 0 0>, rot=<0 0 0 1>}
 upperarm_twist_01_r {parent=upperarm_r, loc=<-0.0041 0 0>, rot=<0 0 0 1>}
 upperarm_twist_02_r {parent=upperarm_r, loc=<-13.7942 0 0>, rot=<0 0 0 1>}
 neck_01 {parent=spine_05, loc=<11.3138 0.1529 0>, rot=<0 0 -0.2003 -0.9797>}
 neck_02 {parent=neck_01, loc=<4.5627 0 0>, rot=<0 0 0.0504 -0.9987>}
 head {parent=neck_02, loc=<5.3914 0 0>, rot=<0 0 0.115 -0.9934>}
 dyn_hair_bk_r_01 {parent=head, loc=<4.2361 -5.7998 5.311>, rot=<0.1088 -0.9169 0.378 0.0674>}
 dyn_hair_bk_r_02 {parent=dyn_hair_bk_r_01, loc=<6.0755 0 0>, rot=<0 0 0.0422 -0.9991>}
 dyn_hair_bk_r_03 {parent=dyn_hair_bk_r_02, loc=<5.6846 0 0>, rot=<0 0 0 -1>}
 dyn_hair_l_01 {parent=head, loc=<4.5776 -1.1945 -6.6264>, rot=<0 0.0386 0 -0.9993>}
 dyn_hair_l_02 {parent=dyn_hair_l_01, loc=<-6.8232 0.0287 -0.7918>, rot=<0 0 0.1302 -0.9915>}
 dyn_hair_l_03 {parent=dyn_hair_l_02, loc=<-4.2709 2.0994 -1.5153>, rot=<0 0 0 -1>}
 dyn_hair_bk_01 {parent=head, loc=<3.9778 -7.144 0>, rot=<0.1514 -0.9885 0 0>}
 dyn_hair_bk_02 {parent=dyn_hair_bk_01, loc=<6.4028 -0.1311 0>, rot=<0 0 0.1276 -0.9918>}
 dyn_hair_bk_03 {parent=dyn_hair_bk_02, loc=<6.4716 0 0>, rot=<0 0 0 1>}
 dyn_hair_r_01 {parent=head, loc=<4.7882 -1.896 7.0366>, rot=<-0.1206 -0.0353 -0.0104 -0.992>}
 dyn_hair_r_02 {parent=dyn_hair_r_01, loc=<-5.4683 0.3186 1.0859>, rot=<0 -0.0762 0 -0.9971>}
 dyn_hair_r_03 {parent=dyn_hair_r_02, loc=<-5.5154 2.933 0.0755>, rot=<0 0 0 1>}
 dyn_hair_bk_l_01 {parent=head, loc=<3.7811 -5.5097 -4.9669>, rot=<0.1821 -0.9262 -0.319 -0.0846>}
 dyn_hair_bk_l_02 {parent=dyn_hair_bk_l_01, loc=<6.0755 0 0>, rot=<0 0 0.0422 -0.9991>}
 dyn_hair_bk_l_03 {parent=dyn_hair_bk_l_02, loc=<5.6846 0 0>, rot=<0 0 0 -1>}
 dyn_hair_strand_frt_01 {parent=head, loc=<11.2232 9.471 1.7294>, rot=<0.4967 0.4254 0.7293 0.201>}
 dyn_hair_strand_frt_02 {parent=dyn_hair_strand_frt_01, loc=<4.4251 0 0>, rot=<0.1369 -0.2006 -0.4456 -0.8617>}
 dyn_hair_strand_frt_03 {parent=dyn_hair_strand_frt_02, loc=<4.0926 0 0>, rot=<-0.0983 -0.3639 -0.1573 -0.9128>}
 dyn_hair_strand_frt_04 {parent=dyn_hair_strand_frt_03, loc=<3.751 0 0>, rot=<-0.0736 0.2653 -0.2053 -0.9392>}
 dyn_hair_strand_rgt_01 {parent=head, loc=<3.8186 9.2336 7.1374>, rot=<0.2801 -0.1793 0.9183 -0.215>}
 dyn_hair_strand_rgt_02 {parent=dyn_hair_strand_rgt_01, loc=<6.5456 0 0>, rot=<0.134 0.0273 0.2413 -0.9608>}
 dyn_hair_strand_rgt_03 {parent=dyn_hair_strand_rgt_02, loc=<5.855 0 0>, rot=<-0.2077 -0.4203 -0.0512 -0.8818>}
 dyn_hair_strand_bk_01_r {parent=head, loc=<14.0086 -3.4593 1.646>, rot=<0.0029 -0.7071 -0.0029 -0.7071>}
 dyn_hair_strand_bk_02_r {parent=dyn_hair_strand_bk_01_r, loc=<-0.4274 -0.1642 2.5197>, rot=<-0.0931 -0.4118 0.3908 -0.8179>}
 dyn_hair_strand_bk_03_r {parent=dyn_hair_strand_bk_02_r, loc=<4.958 0 0>, rot=<0.166 -0.1879 0.1271 -0.9597>}
 dyn_hair_strand_bk_04_r {parent=dyn_hair_strand_bk_03_r, loc=<3.2591 0 0>, rot=<0 0 0 1>}
 dyn_hair_strand_bk_01_l {parent=head, loc=<13.9046 -3.6948 -2.5057>, rot=<0.0029 -0.7071 -0.0029 -0.7071>}
 dyn_hair_strand_bk_02_l {parent=dyn_hair_strand_bk_01_l, loc=<0.839 -2.3432 1.6859>, rot=<0.0247 -0.7232 0.2962 -0.6234>}
 dyn_hair_strand_bk_03_l {parent=dyn_hair_strand_bk_02_l, loc=<5.1499 0 0>, rot=<0.0065 0.2969 0.0192 -0.9547>}
 dyn_hair_strand_bk_04_l {parent=dyn_hair_strand_bk_03_l, loc=<3.0038 0 0>, rot=<0 -0.2088 0 -0.978>}
 faceAttach {parent=head, loc=<-0.7776 -0.1996 0.0118>, rot=<0.0082 -0.7071 -0.0082 -0.7071>, scale=<0.89 0.89 0.89>}
 L_cheek_inner {parent=faceAttach, loc=<4.0946 9.1085 1.5207>, rot=<-0.0556 -0.2319 0.3234 -0.9157>}
 L_lip_upper_outer {parent=faceAttach, loc=<1.7225 10.4841 -0.785>, rot=<0.0486 -0.0371 0.2491 -0.9666>}
 L_lip_corner {parent=faceAttach, loc=<2.5663 9.9892 -0.9495>, rot=<0.0015 0.0095 -0.0145 -0.9998>}
 R_cheek_inner {parent=faceAttach, loc=<-4.0549 9.1163 1.5098>, rot=<-0.0573 0.2361 -0.3229 -0.9147>}
 R_lip_corner {parent=faceAttach, loc=<-2.5177 9.983 -0.8996>, rot=<0.0025 -0.0084 0.0146 -0.9999>}
 C_jaw {parent=faceAttach, loc=<0.0084 3.2511 0.4556>, rot=<0 0 0 1>}
 L_lip_lower_outer {parent=C_jaw, loc=<1.6508 7.2026 -1.3459>, rot=<0.0215 0.0909 0.3402 -0.9357>}
 C_lip_lower_mid {parent=C_jaw, loc=<0.0017 7.881 -1.5218>, rot=<0 0.0019 0.0004 -1>}
 R_lip_lower_outer {parent=C_jaw, loc=<-1.7218 7.2838 -1.4087>, rot=<0.0209 -0.0874 -0.3376 -0.937>}
 dyn_beard_01 {parent=C_jaw, loc=<0.0048 6.2686 -4.2191>, rot=<0.369 -0.369 -0.6032 -0.6032>, scale=<1.1236 1.1236 1.1236>}
 parent_dyn_beard_02 {parent=dyn_beard_01, loc=<0 0 0>, rot=<0 0 0 -1>}
 dyn_beard_02 {parent=parent_dyn_beard_02, loc=<7.1833 0 0.1063>, rot=<0 0 0 -1>}
 dyn_beard_03 {parent=dyn_beard_02, loc=<6.9638 0 0.103>, rot=<0 0 0 -1>}
 dyn_beard_04 {parent=dyn_beard_03, loc=<8.8079 0.5365 -1.8041>, rot=<-0.0015 -0.2084 -0.007 -0.978>}
 dyn_beard_05 {parent=dyn_beard_04, loc=<15.4862 0 0.005>, rot=<0.0011 -0.0025 0 -1>}
 dyn_beard_06 {parent=dyn_beard_05, loc=<14.7438 0 0>, rot=<-0.0011 0 0.0449 -0.999>}
 dyn_beard_07 {parent=dyn_beard_06, loc=<15.3102 0 0>, rot=<0 -0.0001 -0.0449 -0.999>}
 dyn_beard_08 {parent=dyn_beard_07, loc=<13.174 0 0>, rot=<-0.0067 -0.1113 -0.0456 -0.9927>}
 dyn_beard_09 {parent=dyn_beard_08, loc=<13.8109 0 0>, rot=<-0.0029 0.1157 0.1299 -0.9847>}
 dyn_beard_10 {parent=dyn_beard_09, loc=<13.531 0 0>, rot=<-0.0059 0.0817 -0.085 -0.993>}
 dyn_beard_09A {parent=dyn_beard_09, loc=<4.1406 4.6672 2.8078>, rot=<-0.0571 0.082 -0.0837 -0.9915>}
 dyn_beard_08A {parent=dyn_beard_08, loc=<-1.6167 -3.0685 4.8792>, rot=<-0.0012 0.095 0.1177 -0.9885>}
 dyn_beard_braid_B_T3_master_l {parent=dyn_beard_02, loc=<0 -0.1894 -3.7467>, rot=<-0.5614 0.4596 0.392 -0.5657>}
 dyn_beard_braid_B_T3_01_l {parent=dyn_beard_braid_B_T3_master_l, loc=<5.7526 0 0>, rot=<-0.0413 0.1183 -0.0473 -0.991>}
 dyn_beard_braid_B_T3_master_r {parent=dyn_beard_02, loc=<-0.0001 0.1894 -3.7466>, rot=<0.392 -0.5657 0.5614 -0.4596>}
 dyn_beard_braid_B_T3_01_r {parent=dyn_beard_braid_B_T3_master_r, loc=<-5.7527 0.0003 0.0005>, rot=<-0.0413 0.1183 -0.0473 -0.991>}
 dyn_stache01_l {parent=parent_dyn_beard_02, loc=<1.5825 -7.6835 0>, rot=<0.0108 -0.0429 0.0718 -0.9964>}
 dyn_stache02_l {parent=dyn_stache01_l, loc=<8.0961 0 0>, rot=<0 -0.0589 0 -0.9983>}
 dyn_stache03_l {parent=dyn_stache02_l, loc=<6.4489 0 0>, rot=<0 0 0 1>}
 dyn_stache01_r {parent=parent_dyn_beard_02, loc=<1.5827 7.6835 -0.0001>, rot=<0.0718 -0.9964 -0.0108 0.0429>}
 dyn_stache02_r {parent=dyn_stache01_r, loc=<-8.0957 0.0001 -0.0002>, rot=<0 -0.0589 0 -0.9983>}
 dyn_stache03_r {parent=dyn_stache02_r, loc=<-6.4489 0 0>, rot=<0 0 0 -1>}
 dyn_beard_braid_A_T3_master_l {parent=dyn_beard_01, loc=<0 -0.1894 -3.7467>, rot=<-0.5314 0.4435 0.4437 -0.5693>}
 dyn_beard_braid_A_T3_01_l {parent=dyn_beard_braid_A_T3_master_l, loc=<7.4038 0 0>, rot=<0.005 0.0617 0.044 -0.9971>}
 dyn_beard_braid_A_T3_master_r {parent=dyn_beard_01, loc=<-0.0003 0.1894 -3.7465>, rot=<0.4437 -0.5693 0.5314 -0.4435>}
 dyn_beard_braid_A_T3_01_r {parent=dyn_beard_braid_A_T3_master_r, loc=<-7.4038 0 -0.0001>, rot=<0.005 0.0617 0.044 -0.9971>}
 R_lip_upper_outer {parent=faceAttach, loc=<-1.7057 10.5206 -0.7724>, rot=<0.0533 0.0238 -0.2439 -0.968>}
 thigh_l {parent=pelvis, loc=<-2.5693 -0.0109 -7.7501>, rot=<0.0012 0.0572 0.0155 -0.9982>}
 calf_l {parent=thigh_l, loc=<-39.0484 0 0>, rot=<0.0006 -0.0161 -0.0391 -0.9991>}
 calf_twist_02_l {parent=calf_l, loc=<-18.8064 -0.0967 0.0397>, rot=<0.0237 0.0095 -0.0035 -0.9997>}
 foot_l {parent=calf_l, loc=<-42.561 0.1077 0.4292>, rot=<0.0418 -0.0455 0.0343 -0.9975>}
 ball_l {parent=foot_l, loc=<-6.1599 13.8264 0.6806>, rot=<0.0001 -0.0001 -0.7063 -0.7079>}
 thigh_twist_01_l {parent=thigh_l, loc=<-21.6303 0 0>, rot=<0.0013 0 0 -1>}
 thigh_r {parent=pelvis, loc=<-2.5691 -0.0109 7.7501>, rot=<0.0572 -0.0012 0.9982 0.0155>}
 calf_r {parent=thigh_r, loc=<39.0487 0 0>, rot=<0.0006 -0.0161 -0.0391 -0.9991>}
 calf_twist_02_r {parent=calf_r, loc=<18.8064 0.0967 -0.0397>, rot=<0.0237 0.0095 -0.0035 -0.9997>}
 foot_r {parent=calf_r, loc=<42.5611 -0.1077 -0.4292>, rot=<0.0418 -0.0455 0.0343 -0.9975>}
 ball_r {parent=foot_r, loc=<6.1598 -13.8264 -0.6806>, rot=<0.0001 -0.0001 -0.7063 -0.7079>}
 thigh_twist_01_r {parent=thigh_r, loc=<21.6303 0 0>, rot=<0.0013 0 0 -1>}
]
