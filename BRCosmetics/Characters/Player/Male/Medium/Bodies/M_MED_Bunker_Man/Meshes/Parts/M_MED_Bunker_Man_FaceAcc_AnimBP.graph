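# Animation Blueprint that adds secondary motion to the Bunker Man character's hair, beard and moustache.
# A pre-pass uses a faceAttach-to-head constraint, beard Control Rig and Flap nodes that offset back hair from head rotation.
# AnimDynamics, bone constraints and a RigidBody beard simulation run in series, with skydiving forces on the beard.
# `rigids_off` switches between that chain and a cheaper branch, where `bIsSkyDiving` selects cached pose or beard Trail controller.
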AnimBlueprint /BRCosmetics/Characters/Player/Male/Medium/Bodies/M_MED_Bunker_Man/Meshes/Parts/M_MED_Bunker_Man_FaceAcc_AnimBP
target_skeleton=&/M_MED_Bunker_Man_FaceAcc_Skeleton
graphs [
 AnimGraph {nodes [
   n1 = Get hair_alpha() -> (hair_alpha:double)
   Anim Dynamics - Bone: dyn_hair_r_03(Alpha=@n1, ExternalForce="0,0,0") {
    angularDampingOverride=0.9999
    angularSpringConstant=1000
    bAngularSpring=true
    bGravityOverrideInSimSpace=true
    bOverrideAngularDamping=true
    bUsePlanarLimit=false
    bUseSphericalLimits=true
    gravityScale=0.35
    alphaBoolBlend {blendOption=Linear}
    boundBone {boneName=dyn_hair_r_03}
    componentAppliedLinearAccClamp=<1000 1000 1000>
    componentLinearAccScale=<1 1 1>
    componentLinearVelScale=<1 1 1>
    gravityOverride {z=-980}
    physicsBodyDefinitions [{
      collisionType=CustomSphere
      sphereCollisionRadius=1.5
      boundBone {boneName=dyn_hair_r_03}
      boxExtents=<10 30 30>
      constraintSetup {angularLimitsMax=<0 45 5>, angularLimitsMin=<0 -2 -30>, angularTarget=<0 0 0>}
      localJointOffset=<-1.5 0 0>
     }]
    sphericalLimits [
     {limitRadius=60, drivingBone {boneName=spine_05}, sphereLocalOffset=<-48 -6.5 0>}
     {limitRadius=9, drivingBone {boneName=spine_05}, sphereLocalOffset=<13 0 0>}
    ]
   }
   n2 = Anim Dynamics - Bone: dyn_hair_bk_r_01(Alpha=0.55, ExternalForce="0,0,0") {
    angularDampingOverride=0.9999
    angularSpringConstant=5000
    bAngularSpring=true
    bGravityOverrideInSimSpace=true
    bOverrideAngularDamping=true
    bUsePlanarLimit=false
    bUseSphericalLimits=true
    gravityScale=0.35
    alphaBoolBlend {blendOption=Linear}
    boundBone {boneName=dyn_hair_bk_r_01}
    componentAppliedLinearAccClamp=<1200 1200 1000>
    gravityOverride {z=-980}
    physicsBodyDefinitions [{
      collisionType=CustomSphere
      sphereCollisionRadius=4.25
      boundBone {boneName=dyn_hair_bk_r_01}
      boxExtents=<5 10 20>
      constraintSetup {angularLimitsMax=<0 15 20>, angularLimitsMin=<0 -15 -65>, angularTarget=<0 0 0>}
      localJointOffset=<8 0 0>
     }]
    sphericalLimits [{limitRadius=70, drivingBone {boneName=spine_05}, sphereLocalOffset=<-30 50 0>}]
   }
   Anim Dynamics - Bone: dyn_hair_bk_l_01(ComponentPose=@n2, Alpha=0.55, ExternalForce="0,0,0") {
    angularDampingOverride=0.9999
    angularSpringConstant=1000
    bAngularSpring=true
    bGravityOverrideInSimSpace=true
    bOverrideAngularDamping=true
    bUsePlanarLimit=false
    gravityScale=0.35
    alphaBoolBlend {blendOption=Linear}
    boundBone {boneName=dyn_hair_bk_l_01}
    componentAppliedLinearAccClamp=<1200 1200 1000>
    gravityOverride {z=-980}
    physicsBodyDefinitions [{
      collisionType=CustomSphere
      sphereCollisionRadius=3
      boundBone {boneName=dyn_hair_bk_l_01}
      boxExtents=<5 10 20>
      constraintSetup {angularLimitsMax=<0 15 20>, angularLimitsMin=<0 -15 -65>, angularTarget=<0 0 0>}
      localJointOffset=<6.5 0 0>
     }]
    sphericalLimits [{limitRadius=8, drivingBone {boneName=neck_02}, sphereLocalOffset=<0 0 0>}]
   }
   n3 = Trail controller - Bone: dyn_beard_10(ComponentPose=Local To Component(LocalPose=Use cached pose 'AnimCache'()), Alpha=1) {
    bActorSpaceFakeVel=true
    bUsePlanarLimit=true
    chainLength=7
    baseJoint {boneName=root}
    planarLimits [
     {drivingBone {boneName=pelvis}, planeTransform {location=<-25 15 0>, rotation=<rot 0 0 89.9999>, scale=<1 1 1>}}
     {drivingBone {boneName=pelvis}, planeTransform {location=<-35 20 0>, rotation=<rot 0 -20 89.9999>, scale=<1 1 1>}}
    ]
    rotationLimits [
     {limitMax=<180 180 180>}
     {limitMax=<180 180 180>}
     {limitMax=<180 180 180>}
     {limitMax=<180 180 180>}
     {limitMax=<180 180 180>}
     {limitMax=<180 180 180>}
     {limitMax=<180 180 180>}
    ]
    rotationOffsets [<0 0 0>, <0 0 0>, <0 0 0>, <0 0 0>, <0 0 0>, <0 0 0>, <0 0 0>]
    trailBone {boneName=dyn_beard_10}
   }
   n4 = Component To Local(ComponentPose=@n3)
   n5 = Use cached pose 'AnimCache'()
   n6 = Get bIsSkyDiving()
   n7 = Anim Dynamics - Bone: dyn_hair_bk_01(ComponentPose=Local To Component(LocalPose=Use cached pose 'AnimCache'()), Alpha=@n1, ExternalForce="0,0,0") {
    angularDampingOverride=0.9999
    angularSpringConstant=5000
    bAngularSpring=true
    bGravityOverrideInSimSpace=true
    bOverrideAngularDamping=true
    bUsePlanarLimit=false
    bUseSphericalLimits=true
    gravityScale=0.35
    alphaBoolBlend {blendOption=Linear}
    boundBone {boneName=dyn_hair_bk_01}
    componentAppliedLinearAccClamp=<1600 1600 1000>
    componentLinearAccScale=<1 1 1>
    componentLinearVelScale=<1 1 1>
    gravityOverride {z=-980}
    physicsBodyDefinitions [{
      collisionType=CustomSphere
      sphereCollisionRadius=4.25
      boundBone {boneName=dyn_hair_bk_01}
      boxExtents=<5 10 20>
      constraintSetup {angularLimitsMax=<0 15 25>, angularLimitsMin=<0 -15 -65>, angularTarget=<0 0 0>}
      localJointOffset=<8 0 0>
     }]
    planarLimits [{drivingBone {boneName=neck_01}, planeTransform {location=<15 -6 6>, rotation=<rot 90 -90 0>}}]
    sphericalLimits [{limitRadius=70, drivingBone {boneName=spine_05}, sphereLocalOffset=<-30 50 0>}]
   }
   n8 = Anim Dynamics - Bone: dyn_hair_bk_02(ComponentPose=@n7, Alpha=@n1, ExternalForce="0,0,0") {
    angularDampingOverride=1
    angularSpringConstant=1000
    bAngularSpring=true
    bGravityOverrideInSimSpace=true
    bOverrideAngularDamping=true
    bUseSphericalLimits=true
    gravityScale=0.35
    alphaBoolBlend {blendOption=Linear}
    boundBone {boneName=dyn_hair_bk_02}
    componentAppliedLinearAccClamp=<1600 1600 1100>
    componentLinearAccScale=<1 1 1>
    componentLinearVelScale=<1 1 1>
    gravityOverride {z=-980}
    physicsBodyDefinitions [{
      collisionType=CustomSphere
      sphereCollisionRadius=2.5
      boundBone {boneName=dyn_hair_bk_02}
      boxExtents=<5 20 20>
      constraintSetup {angularLimitsMax=<0 15 7>, angularLimitsMin=<0 -15 -75>, angularTarget=<1 0 0>}
      localJointOffset=<12 0 0>
     }]
    planarLimits [
     {drivingBone {boneName=neck_01}, planeTransform {location=<7 -8 6>, rotation=<rot -90 -90 0>}}
     {drivingBone {boneName=spine_05}, planeTransform {location=<7 -12 6>, rotation=<rot -90 -36.5651 0>}}
    ]
    sphericalLimits [
     {limitRadius=45, drivingBone {boneName=spine_05}, sphereLocalOffset=<-25 18 0>}
     {limitRadius=8, drivingBone {boneName=neck_02}, sphereLocalOffset=<0 -1.5 0>}
    ]
   }
   n9 = "Constraint - dyn_hair_bk_l_01 (P)"(ComponentPose=@n8, Alpha=1, ConstraintWeights_0=1) {
    alphaBoolBlend {blendOption=Linear}
    boneToModify {boneName=dyn_hair_bk_l_01}
    constraintSetup [{targetBone {boneName=dyn_hair_bk_01}}]
    constraintWeights [1]
   }
   n10 = "Constraint - dyn_hair_bk_l_02 (P)"(ComponentPose=@n9, Alpha=1, ConstraintWeights_0=1) {
    alphaBoolBlend {blendOption=Linear}
    boneToModify {boneName=dyn_hair_bk_l_02}
    constraintSetup [{targetBone {boneName=dyn_hair_bk_02}}]
    constraintWeights [1]
   }
   n11 = "Constraint - dyn_hair_bk_r_01 (P)"(ComponentPose=@n10, Alpha=1, ConstraintWeights_0=1) {
    alphaBoolBlend {blendOption=Linear}
    boneToModify {boneName=dyn_hair_bk_r_01}
    constraintSetup [{targetBone {boneName=dyn_hair_bk_01}}]
    constraintWeights [1]
   }
   n12 = "Constraint - dyn_hair_bk_r_02 (P)"(ComponentPose=@n11, Alpha=1, ConstraintWeights_0=1) {
    alphaBoolBlend {blendOption=Linear}
    boneToModify {boneName=dyn_hair_bk_r_02}
    constraintSetup [{targetBone {boneName=dyn_hair_bk_02}}]
    constraintWeights [1]
   }
   n13 = Anim Dynamics - Bone: dyn_hair_bk_r_02(ComponentPose=@n12, Alpha=0.55, ExternalForce="0,0,0") {
    angularDampingOverride=0.9999
    angularSpringConstant=1000
    bAngularSpring=true
    bGravityOverrideInSimSpace=true
    bOverrideAngularDamping=true
    bUseSphericalLimits=true
    gravityScale=0.35
    alphaBoolBlend {blendOption=Linear}
    boundBone {boneName=dyn_hair_bk_r_02}
    componentAppliedLinearAccClamp=<1200 1200 1000>
    gravityOverride {z=-980}
    physicsBodyDefinitions [{
      collisionType=CustomSphere
      sphereCollisionRadius=4
      boundBone {boneName=dyn_hair_bk_r_02}
      boxExtents=<5 20 20>
      constraintSetup {angularLimitsMax=<0 15 7>, angularLimitsMin=<0 -15 -75>, angularTarget=<0 0 0>}
      localJointOffset=<10 0 0>
     }]
    sphericalLimits [{limitRadius=55, drivingBone {boneName=spine_05}, sphereLocalOffset=<-40 23 8>}]
   }
   n14 = Anim Dynamics - Bone: dyn_hair_bk_l_02(ComponentPose=@n13, Alpha=0.55, ExternalForce="0,0,0") {
    angularDampingOverride=0.9999
    angularSpringConstant=1000
    bAngularSpring=true
    bGravityOverrideInSimSpace=true
    bOverrideAngularDamping=true
    bUsePlanarLimit=false
    bUseSphericalLimits=true
    gravityScale=0.35
    alphaBoolBlend {blendOption=Linear}
    boundBone {boneName=dyn_hair_bk_l_02}
    componentAppliedLinearAccClamp=<1200 1200 1000>
    gravityOverride {z=-980}
    physicsBodyDefinitions [{
      collisionType=CustomSphere
      sphereCollisionRadius=3
      boundBone {boneName=dyn_hair_bk_l_02}
      boxExtents=<5 10 20>
      constraintSetup {angularLimitsMax=<0 15 7>, angularLimitsMin=<0 -15 -75>, angularTarget=<0 0 0>}
      localJointOffset=<10 0 0>
     }]
    sphericalLimits [{limitRadius=46, drivingBone {boneName=spine_05}, sphereLocalOffset=<-30 19 -8>}]
   }
   n15 = Anim Dynamics - Bone: dyn_hair_l_01(ComponentPose=@n14, Alpha=@n1, ExternalForce="0,0,0") {
    angularDampingOverride=0.9999
    angularSpringConstant=1000
    bAngularSpring=true
    bGravityOverrideInSimSpace=true
    bOverrideAngularDamping=true
    bUsePlanarLimit=false
    gravityScale=0.35
    alphaBoolBlend {blendOption=Linear}
    boundBone {boneName=dyn_hair_l_01}
    componentAppliedLinearAccClamp=<1000 1000 1000>
    componentLinearAccScale=<1 1 1>
    componentLinearVelScale=<1 1 1>
    gravityOverride {z=-980}
    physicsBodyDefinitions [{
      collisionType=CustomSphere
      sphereCollisionRadius=3
      boundBone {boneName=dyn_hair_l_01}
      boxExtents=<5 30 30>
      constraintSetup {angularLimitsMax=<0 8 3>, angularLimitsMin=<0 -8 -11>, angularTarget=<0 0 0>}
      localJointOffset=<-6.5 0 0>
     }]
    sphericalLimits [
     {limitRadius=8, drivingBone {boneName=neck_02}, sphereLocalOffset=<0 0 0>}
     {limitRadius=60, drivingBone {boneName=spine_05}, sphereLocalOffset=<-42 5 0>}
    ]
   }
   n16 = Anim Dynamics - Bone: dyn_hair_l_02(ComponentPose=@n15, Alpha=@n1, ExternalForce="0,0,0") {
    angularDampingOverride=0.9999
    angularSpringConstant=200
    bAngularSpring=true
    bGravityOverrideInSimSpace=true
    bOverrideAngularDamping=true
    bUseSphericalLimits=true
    gravityScale=0.35
    alphaBoolBlend {blendOption=Linear}
    boundBone {boneName=dyn_hair_l_02}
    componentAppliedLinearAccClamp=<1000 1000 1000>
    componentLinearAccScale=<1 1 1>
    componentLinearVelScale=<1 1 1>
    gravityOverride {z=-980}
    physicsBodyDefinitions [{
      collisionType=CustomSphere
      sphereCollisionRadius=2.25
      boundBone {boneName=dyn_hair_l_02}
      boxExtents=<10 30 30>
      constraintSetup {angularLimitsMax=<0 8 40>, angularLimitsMin=<0 -8 -15>, angularTarget=<0 0 0>}
      localJointOffset=<-6.5 0 0>
     }]
    planarLimits [{drivingBone {boneName=clavicle_l}, planeTransform {location=<0 -2 0>, rotation=<rot -90 90 0>}}]
    sphericalLimits [
     {limitRadius=59, drivingBone {boneName=spine_05}, sphereLocalOffset=<-48 -6.5 0>}
     {limitRadius=6, drivingBone {boneName=spine_05}, sphereLocalOffset=<13 0 0>}
    ]
   }
   n17 = Anim Dynamics - Bone: dyn_hair_l_03(ComponentPose=@n16, Alpha=@n1, ExternalForce="0,0,0") {
    angularDampingOverride=0.9999
    angularSpringConstant=1000
    bAngularSpring=true
    bGravityOverrideInSimSpace=true
    bOverrideAngularDamping=true
    bUsePlanarLimit=false
    bUseSphericalLimits=true
    gravityScale=0.35
    alphaBoolBlend {blendOption=Linear}
    boundBone {boneName=dyn_hair_l_03}
    componentAppliedLinearAccClamp=<1000 1000 1000>
    componentLinearAccScale=<1 1 1>
    componentLinearVelScale=<1 1 1>
    gravityOverride {z=-980}
    physicsBodyDefinitions [{
      collisionType=CustomSphere
      sphereCollisionRadius=3
      boundBone {boneName=dyn_hair_l_03}
      boxExtents=<5 30 30>
      constraintSetup {angularLimitsMax=<0 10 45>, angularLimitsMin=<0 -10 -25>, angularTarget=<0 0 0>}
      localJointOffset=<-6.5 0 0>
     }]
    sphericalLimits [
     {limitRadius=60, drivingBone {boneName=spine_05}, sphereLocalOffset=<-48 -6.5 0>}
     {limitRadius=8, drivingBone {boneName=spine_05}, sphereLocalOffset=<13 0 0>}
    ]
   }
   n18 = Anim Dynamics - Bone: dyn_hair_r_01(ComponentPose=@n17, Alpha=@n1, ExternalForce="0,0,0") {
    angularDampingOverride=0.9999
    angularSpringConstant=1000
    bAngularSpring=true
    bGravityOverrideInSimSpace=true
    bOverrideAngularDamping=true
    bUsePlanarLimit=false
    gravityScale=0.35
    alphaBoolBlend {blendOption=Linear}
    boundBone {boneName=dyn_hair_r_01}
    componentAppliedLinearAccClamp=<1000 1000 1000>
    componentLinearAccScale=<1 1 1>
    componentLinearVelScale=<1 1 1>
    gravityOverride {z=-980}
    physicsBodyDefinitions [{
      collisionType=CustomSphere
      sphereCollisionRadius=3
      boundBone {boneName=dyn_hair_r_01}
      boxExtents=<5 30 30>
      constraintSetup {angularLimitsMax=<0 20 17>, angularLimitsMin=<0 -2.5 -8>, angularTarget=<0 0 0>}
      localJointOffset=<-6.5 0 0>
     }]
    sphericalLimits [
     {limitRadius=8, drivingBone {boneName=neck_02}, sphereLocalOffset=<0 0 0>}
     {limitRadius=8, drivingBone {boneName=neck_01}, sphereLocalOffset=<5 10 0>}
    ]
   }
   n19 = Anim Dynamics - Bone: dyn_hair_r_02(ComponentPose=@n18, Alpha=@n1, ExternalForce="0,0,0") {
    angularDampingOverride=0.9999
    angularSpringConstant=200
    bAngularSpring=true
    bGravityOverrideInSimSpace=true
    bOverrideAngularDamping=true
    bUseSphericalLimits=true
    gravityScale=0.35
    alphaBoolBlend {blendOption=Linear}
    boundBone {boneName=dyn_hair_r_02}
    componentAppliedLinearAccClamp=<1000 1000 1000>
    componentLinearAccScale=<1 1 1>
    componentLinearVelScale=<1 1 1>
    gravityOverride {z=-980}
    physicsBodyDefinitions [{
      collisionType=CustomSphere
      sphereCollisionRadius=2.25
      boundBone {boneName=dyn_hair_r_02}
      boxExtents=<10 30 30>
      constraintSetup {angularLimitsMax=<0 20 20>, angularLimitsMin=<0 -5 -20>, angularTarget=<0 0 0>}
      localJointOffset=<-6.5 0 0>
     }]
    planarLimits [{drivingBone {boneName=clavicle_l}, planeTransform {location=<0 -2 0>, rotation=<rot -90 90 0>}}]
    sphericalLimits [
     {limitRadius=60, drivingBone {boneName=spine_05}, sphereLocalOffset=<-48 -6.5 0>}
     {limitRadius=6.5, drivingBone {boneName=spine_05}, sphereLocalOffset=<13 0 0>}
    ]
   }
   n20 = "Transform (Modify) Bone - Bone: dyn_hair_r_02"(ComponentPose=@n19, Alpha=1, Translation="0,0,1", Rotation="-5,-5,-8", Scale="1,1,1") {
    rotationMode=BMM_Additive
    rotationSpace=BCS_BoneSpace
    translationMode=BMM_Additive
    alphaBoolBlend {blendOption=Linear}
    boneToModify {boneName=dyn_hair_r_02}
    rotation=<rot -5 -5 -8>
    translation {z=1}
   }
   n21 = Anim Dynamics - Bone: dyn_hair_strand_frt_01(ComponentPose=@n20, Alpha=1, ExternalForce="0,0,0") {
    angularDampingOverride=1
    angularSpringConstant=1500
    bAngularSpring=true
    bGravityOverrideInSimSpace=true
    bOverrideAngularDamping=true
    gravityScale=0.35
    simulationSpace=RootRelative
    alphaBoolBlend {blendOption=Linear}
    boundBone {boneName=dyn_hair_strand_frt_01}
    physicsBodyDefinitions [{
      sphereCollisionRadius=0
      boundBone {boneName=dyn_hair_strand_frt_01}
      boxExtents=<10 20 20>
      constraintSetup {angularLimitsMax=<0 5 3>, angularLimitsMin=<0 -5 -10>, angularTarget=<0 0 0>}
      localJointOffset=<2 0 0>
     }]
   }
   n22 = Anim Dynamics - Bone: dyn_hair_strand_frt_02(ComponentPose=@n21, Alpha=1, ExternalForce="0,0,0") {
    angularDampingOverride=1
    angularSpringConstant=1500
    bAngularSpring=true
    bGravityOverrideInSimSpace=true
    bOverrideAngularDamping=true
    gravityScale=0.35
    simulationSpace=RootRelative
    alphaBoolBlend {blendOption=Linear}
    boundBone {boneName=dyn_hair_strand_frt_02}
    physicsBodyDefinitions [{
      sphereCollisionRadius=0
      boundBone {boneName=dyn_hair_strand_frt_02}
      boxExtents=<10 20 20>
      constraintSetup {angularLimitsMax=<0 5 0>, angularLimitsMin=<0 -5 -40>, angularTarget=<0 0 0>}
      localJointOffset=<10 0 0>
     }]
   }
   n23 = Anim Dynamics - Bone: dyn_hair_strand_frt_03(ComponentPose=@n22, Alpha=1, ExternalForce="0,0,0") {
    angularDampingOverride=1
    angularSpringConstant=1500
    bAngularSpring=true
    bGravityOverrideInSimSpace=true
    bOverrideAngularDamping=true
    gravityScale=0.35
    simulationSpace=RootRelative
    alphaBoolBlend {blendOption=Linear}
    boundBone {boneName=dyn_hair_strand_frt_03}
    physicsBodyDefinitions [{
      sphereCollisionRadius=0
      boundBone {boneName=dyn_hair_strand_frt_03}
      boxExtents=<10 20 20>
      constraintSetup {angularLimitsMax=<0 0 0>, angularLimitsMin=<0 0 -40>, angularTarget=<0 0 0>}
      localJointOffset=<10 0 0>
     }]
   }
   n24 = Anim Dynamics - Bone: dyn_hair_strand_frt_04(ComponentPose=@n23, Alpha=1, ExternalForce="0,0,0") {
    angularDampingOverride=1
    angularSpringConstant=1500
    bAngularSpring=true
    bGravityOverrideInSimSpace=true
    bOverrideAngularDamping=true
    gravityScale=0.35
    simulationSpace=RootRelative
    alphaBoolBlend {blendOption=Linear}
    boundBone {boneName=dyn_hair_strand_frt_04}
    physicsBodyDefinitions [{
      sphereCollisionRadius=0
      boundBone {boneName=dyn_hair_strand_frt_04}
      boxExtents=<10 20 20>
      constraintSetup {angularLimitsMax=<0 0 0>, angularLimitsMin=<0 0 -40>, angularTarget=<0 0 0>}
      localJointOffset=<10 0 0>
     }]
   }
   n25 = Anim Dynamics - Bone: dyn_hair_strand_rgt_01(ComponentPose=@n24, Alpha=1, ExternalForce="0,0,0") {
    angularDampingOverride=1
    angularSpringConstant=1500
    bAngularSpring=true
    bGravityOverrideInSimSpace=true
    bOverrideAngularDamping=true
    gravityScale=0.35
    simulationSpace=RootRelative
    alphaBoolBlend {blendOption=Linear}
    boundBone {boneName=dyn_hair_strand_rgt_01}
    physicsBodyDefinitions [{
      sphereCollisionRadius=0
      boundBone {boneName=dyn_hair_strand_rgt_01}
      boxExtents=<8 15 25>
      constraintSetup {angularLimitsMax=<0 10 0>, angularLimitsMin=<0 -35 -40>, angularTarget=<0 0 0>}
      localJointOffset=<10 0 0>
     }]
   }
   n26 = Anim Dynamics - Bone: dyn_hair_strand_rgt_02(ComponentPose=@n25, Alpha=1, ExternalForce="0,0,0") {
    angularDampingOverride=1
    angularSpringConstant=1500
    bAngularSpring=true
    bGravityOverrideInSimSpace=true
    bOverrideAngularDamping=true
    gravityScale=0.35
    simulationSpace=RootRelative
    alphaBoolBlend {blendOption=Linear}
    boundBone {boneName=dyn_hair_strand_rgt_02}
    physicsBodyDefinitions [{
      sphereCollisionRadius=0
      boundBone {boneName=dyn_hair_strand_rgt_02}
      boxExtents=<10 20 20>
      constraintSetup {angularLimitsMax=<0 10 10>, angularLimitsMin=<0 -30 -40>, angularTarget=<0 0 0>}
      localJointOffset=<10 0 0>
     }]
   }
   n27 = Anim Dynamics - Bone: dyn_hair_strand_rgt_03(ComponentPose=@n26, Alpha=1, ExternalForce="0,0,0") {
    angularDampingOverride=1
    angularSpringConstant=1500
    bAngularSpring=true
    bGravityOverrideInSimSpace=true
    bOverrideAngularDamping=true
    gravityScale=0.35
    simulationSpace=RootRelative
    alphaBoolBlend {blendOption=Linear}
    boundBone {boneName=dyn_hair_strand_rgt_03}
    physicsBodyDefinitions [{
      sphereCollisionRadius=0
      boundBone {boneName=dyn_hair_strand_rgt_03}
      boxExtents=<10 20 20>
      constraintSetup {angularLimitsMax=<0 10 15>, angularLimitsMin=<0 -30 -40>, angularTarget=<0 0 0>}
      localJointOffset=<10 0 0>
     }]
   }
   n28 = Anim Dynamics - Bone: dyn_hair_strand_bk_01_l(ComponentPose=@n27, Alpha=1, ExternalForce="0,0,0") {
    angularDampingOverride=1
    angularSpringConstant=1500
    bAngularSpring=true
    bGravityOverrideInSimSpace=true
    bOverrideAngularDamping=true
    gravityScale=0.35
    simulationSpace=RootRelative
    alphaBoolBlend {blendOption=Linear}
    boundBone {boneName=dyn_hair_strand_bk_01_l}
    physicsBodyDefinitions [{
      sphereCollisionRadius=0
      boundBone {boneName=dyn_hair_strand_bk_01_l}
      boxExtents=<30 20 10>
      constraintSetup {angularLimitsMax=<20 20 10>, angularLimitsMin=<-45 -20 -40>, angularTarget=<0 0 0>}
      localJointOffset=<0.25 0 -3>
     }]
   }
   n29 = Anim Dynamics - Bone: dyn_hair_strand_bk_02_l(ComponentPose=@n28, Alpha=1, ExternalForce="0,0,0") {
    angularDampingOverride=1
    angularSpringConstant=1500
    bAngularSpring=true
    bGravityOverrideInSimSpace=true
    bOverrideAngularDamping=true
    bUseSphericalLimits=true
    gravityScale=0.35
    simulationSpace=RootRelative
    alphaBoolBlend {blendOption=Linear}
    boundBone {boneName=dyn_hair_strand_bk_02_l}
    physicsBodyDefinitions [{
      collisionType=CustomSphere
      sphereCollisionRadius=2
      boundBone {boneName=dyn_hair_strand_bk_02_l}
      boxExtents=<10 20 20>
      constraintSetup {angularLimitsMax=<0 20 10>, angularLimitsMin=<0 -20 -40>, angularTarget=<0 0 0>}
      localJointOffset=<10 0 0>
     }]
    sphericalLimits [{limitRadius=9, drivingBone {boneName=head}, sphereLocalOffset=<0 0 0>}]
   }
   n30 = Anim Dynamics - Bone: dyn_hair_strand_bk_03_l(ComponentPose=@n29, Alpha=1, ExternalForce="0,0,0") {
    angularDampingOverride=1
    angularSpringConstant=1500
    bAngularSpring=true
    bGravityOverrideInSimSpace=true
    bOverrideAngularDamping=true
    gravityScale=0.35
    simulationSpace=RootRelative
    alphaBoolBlend {blendOption=Linear}
    boundBone {boneName=dyn_hair_strand_bk_03_l}
    physicsBodyDefinitions [{
      sphereCollisionRadius=0
      boundBone {boneName=dyn_hair_strand_bk_03_l}
      boxExtents=<10 20 20>
      constraintSetup {angularLimitsMax=<0 20 10>, angularLimitsMin=<0 -20 -40>, angularTarget=<0 0 0>}
      localJointOffset=<10 0 0>
     }]
   }
   n31 = Anim Dynamics - Bone: dyn_hair_strand_bk_04_l(ComponentPose=@n30, Alpha=1, ExternalForce="0,0,0") {
    angularDampingOverride=1
    angularSpringConstant=1500
    bAngularSpring=true
    bGravityOverrideInSimSpace=true
    bOverrideAngularDamping=true
    gravityScale=0.35
    simulationSpace=RootRelative
    alphaBoolBlend {blendOption=Linear}
    boundBone {boneName=dyn_hair_strand_bk_04_l}
    physicsBodyDefinitions [{
      sphereCollisionRadius=0
      boundBone {boneName=dyn_hair_strand_bk_04_l}
      boxExtents=<10 20 20>
      constraintSetup {angularLimitsMax=<0 20 10>, angularLimitsMin=<0 -20 -40>, angularTarget=<0 0 0>}
      localJointOffset=<10 0 0>
     }]
   }
   n32 = Anim Dynamics - Bone: dyn_hair_strand_bk_01_r(ComponentPose=@n31, Alpha=1, ExternalForce="0,0,0") {
    angularDampingOverride=1
    angularSpringConstant=1500
    bAngularSpring=true
    bGravityOverrideInSimSpace=true
    bOverrideAngularDamping=true
    gravityScale=0.32
    simulationSpace=RootRelative
    alphaBoolBlend {blendOption=Linear}
    boundBone {boneName=dyn_hair_strand_bk_01_r}
    physicsBodyDefinitions [{
      sphereCollisionRadius=0
      boundBone {boneName=dyn_hair_strand_bk_01_r}
      boxExtents=<30 20 10>
      constraintSetup {angularLimitsMax=<0 20 10>, angularLimitsMin=<-35 -20 -40>, angularTarget=<0 0 0>}
      localJointOffset=<0.25 0 -3>
     }]
   }
   n33 = Anim Dynamics - Bone: dyn_hair_strand_bk_02_r(ComponentPose=@n32, Alpha=1, ExternalForce="0,0,0") {
    angularDampingOverride=1
    angularSpringConstant=1500
    bAngularSpring=true
    bGravityOverrideInSimSpace=true
    bOverrideAngularDamping=true
    bUseSphericalLimits=true
    gravityScale=0.35
    simulationSpace=RootRelative
    alphaBoolBlend {blendOption=Linear}
    boundBone {boneName=dyn_hair_strand_bk_02_r}
    physicsBodyDefinitions [{
      collisionType=CustomSphere
      sphereCollisionRadius=2
      boundBone {boneName=dyn_hair_strand_bk_02_r}
      boxExtents=<10 35 20>
      constraintSetup {angularLimitsMax=<0 20 0>, angularLimitsMin=<0 -20 -40>, angularTarget=<0 0 0>}
      localJointOffset=<10 0 0>
     }]
    sphericalLimits [{limitRadius=10, drivingBone {boneName=head}, sphereLocalOffset=<0 0 0>}]
   }
   n34 = Anim Dynamics - Bone: dyn_hair_strand_bk_03_r(ComponentPose=@n33, Alpha=1, ExternalForce="0,0,0") {
    angularDampingOverride=1
    angularSpringConstant=1500
    bAngularSpring=true
    bGravityOverrideInSimSpace=true
    bOverrideAngularDamping=true
    gravityScale=0.38
    simulationSpace=RootRelative
    alphaBoolBlend {blendOption=Linear}
    boundBone {boneName=dyn_hair_strand_bk_03_r}
    physicsBodyDefinitions [{
      sphereCollisionRadius=0
      boundBone {boneName=dyn_hair_strand_bk_03_r}
      boxExtents=<10 25 25>
      constraintSetup {angularLimitsMax=<0 20 10>, angularLimitsMin=<0 -20 -40>, angularTarget=<0 0 0>}
      localJointOffset=<5 0 0>
     }]
   }
   n35 = Anim Dynamics - Bone: dyn_hair_strand_bk_04_r(ComponentPose=@n34, Alpha=1, ExternalForce="0,0,0") {
    angularDampingOverride=1
    angularSpringConstant=1500
    bAngularSpring=true
    bGravityOverrideInSimSpace=true
    bOverrideAngularDamping=true
    gravityScale=0.35
    simulationSpace=RootRelative
    alphaBoolBlend {blendOption=Linear}
    boundBone {boneName=dyn_hair_strand_bk_04_r}
    physicsBodyDefinitions [{
      sphereCollisionRadius=0
      boundBone {boneName=dyn_hair_strand_bk_04_r}
      boxExtents=<10 20 20>
      constraintSetup {angularLimitsMax=<0 20 10>, angularLimitsMin=<0 -20 -40>, angularTarget=<0 0 0>}
      localJointOffset=<5 0 0>
     }]
   }
   n36 = Get skydiving_force() -> (skydiving_force:Vector)
   n37 = Anim Dynamics - Bone: dyn_beard_02(ComponentPose=@n35, Alpha=0.1, ExternalForce=@n36) {
    angularDampingOverride=1
    angularSpringConstant=100
    bAngularSpring=true
    bGravityOverrideInSimSpace=true
    bOverrideAngularDamping=true
    gravityScale=0.33
    numSolverIterationsPostUpdate=2
    numSolverIterationsPreUpdate=8
    simulationSpace=RootRelative
    alphaBoolBlend {blendOption=Linear}
    boundBone {boneName=dyn_beard_02}
    physicsBodyDefinitions [{
      sphereCollisionRadius=0
      boundBone {boneName=dyn_beard_02}
      boxExtents=<10 30 30>
      constraintSetup {angularLimitsMax=<0 5 25>, angularLimitsMin=<0 -85 -25>, angularTarget=<0 0 0>}
      localJointOffset=<7 0 0>
     }]
   }
   n38 = Anim Dynamics - Bone: dyn_beard_03(ComponentPose=@n37, Alpha=0.3, ExternalForce=@n36) {
    angularDampingOverride=1
    angularSpringConstant=100
    bAngularSpring=true
    bGravityOverrideInSimSpace=true
    bOverrideAngularDamping=true
    gravityScale=0.33
    numSolverIterationsPostUpdate=2
    numSolverIterationsPreUpdate=8
    simulationSpace=RootRelative
    alphaBoolBlend {blendOption=Linear}
    boundBone {boneName=dyn_beard_03}
    physicsBodyDefinitions [{
      sphereCollisionRadius=0
      boundBone {boneName=dyn_beard_03}
      boxExtents=<10 30 30>
      constraintSetup {angularLimitsMax=<0 5 25>, angularLimitsMin=<0 -85 -25>, angularTarget=<0 0 0>}
      localJointOffset=<7 0 0>
     }]
   }
   n39 = Anim Dynamics - Bone: dyn_beard_04(ComponentPose=@n38, Alpha=0.5, ExternalForce=@n36) {
    angularDampingOverride=1
    angularSpringConstant=100
    bAngularSpring=true
    bGravityOverrideInSimSpace=true
    bOverrideAngularDamping=true
    gravityScale=0.33
    numSolverIterationsPostUpdate=2
    numSolverIterationsPreUpdate=8
    simulationSpace=RootRelative
    alphaBoolBlend {blendOption=Linear}
    boundBone {boneName=dyn_beard_04}
    componentLinearAccScale=<1 1 1>
    componentLinearVelScale=<1 1 1>
    physicsBodyDefinitions [{
      sphereCollisionRadius=0
      boundBone {boneName=dyn_beard_04}
      boxExtents=<10 30 30>
      constraintSetup {angularLimitsMax=<0 5 25>, angularLimitsMin=<0 -85 -25>, angularTarget=<0 0 0>}
      localJointOffset=<7 0 0>
     }]
   }
   n40 = Anim Dynamics - Bone: dyn_beard_05(ComponentPose=@n39, Alpha=0.75, ExternalForce=@n36) {
    angularDampingOverride=1
    angularSpringConstant=100
    bAngularSpring=true
    bGravityOverrideInSimSpace=true
    bOverrideAngularDamping=true
    gravityScale=0.33
    numSolverIterationsPostUpdate=2
    numSolverIterationsPreUpdate=8
    simulationSpace=RootRelative
    alphaBoolBlend {blendOption=Linear}
    boundBone {boneName=dyn_beard_05}
    componentLinearAccScale=<0.5 0.5 0.5>
    componentLinearVelScale=<0.5 0.5 0.5>
    physicsBodyDefinitions [{
      sphereCollisionRadius=0
      boundBone {boneName=dyn_beard_05}
      boxExtents=<10 30 30>
      constraintSetup {angularLimitsMax=<0 5 25>, angularLimitsMin=<0 -85 -25>, angularTarget=<0 0 0>}
      localJointOffset=<7 0 0>
     }]
   }
   n41 = Anim Dynamics - Bone: dyn_beard_06(ComponentPose=@n40, Alpha=0.8, ExternalForce=@n36) {
    angularDampingOverride=1
    angularSpringConstant=100
    bAngularSpring=true
    bGravityOverrideInSimSpace=true
    bOverrideAngularDamping=true
    gravityScale=0.33
    numSolverIterationsPostUpdate=2
    numSolverIterationsPreUpdate=8
    simulationSpace=RootRelative
    alphaBoolBlend {blendOption=Linear}
    boundBone {boneName=dyn_beard_06}
    physicsBodyDefinitions [{
      sphereCollisionRadius=0
      boundBone {boneName=dyn_beard_06}
      boxExtents=<10 30 30>
      constraintSetup {angularLimitsMax=<0 5 25>, angularLimitsMin=<0 -85 -25>, angularTarget=<0 0 0>}
      localJointOffset=<7 0 0>
     }]
   }
   n42 = "Anim Dynamics - Chain: dyn_stache01_l -> dyn_stache03_l"(ComponentPose=@n41, Alpha=0.55, ExternalForce="0,0,0") {
    angularDampingOverride=1
    angularSpringConstant=1000
    bAngularSpring=true
    bChain=true
    bGravityOverrideInSimSpace=true
    bOverrideAngularDamping=true
    bUseSphericalLimits=true
    gravityScale=0.35
    numSolverIterationsPostUpdate=2
    numSolverIterationsPreUpdate=8
    simulationSpace=RootRelative
    alphaBoolBlend {blendOption=Linear}
    boundBone {boneName=dyn_stache01_l}
    chainEnd {boneName=dyn_stache03_l}
    componentAppliedLinearAccClamp=<1000 1000 1000>
    componentLinearAccScale=<1 1 1>
    componentLinearVelScale=<1 1 1>
    physicsBodyDefinitions [{
      collisionType=CustomSphere
      sphereCollisionRadius=2
      boundBone {boneName=dyn_stache01_l}
      boxExtents=<10 20 20>
      constraintSetup {angularLimitsMax=<0 5 3>, angularLimitsMin=<0 -85 -25>, angularTarget=<0 0 0>}
      localJointOffset=<3 0 0>
     }]
    sphericalLimits [{limitRadius=55, drivingBone {boneName=spine_04}, sphereLocalOffset=<0 -43 0>}]
   }
   n43 = "Anim Dynamics - Chain: dyn_stache01_r -> dyn_stache03_r"(ComponentPose=@n42, Alpha=0.6, ExternalForce="0,0,0") {
    angularDampingOverride=1
    angularSpringConstant=50000
    bChain=true
    bGravityOverrideInSimSpace=true
    bOverrideAngularDamping=true
    bUseSphericalLimits=true
    gravityScale=0.33
    numSolverIterationsPostUpdate=2
    numSolverIterationsPreUpdate=8
    simulationSpace=RootRelative
    alphaBoolBlend {blendOption=Linear}
    boundBone {boneName=dyn_stache01_r}
    chainEnd {boneName=dyn_stache03_r}
    componentAppliedLinearAccClamp=<1000 1000 1000>
    componentLinearAccScale=<0.5 0.5 0.5>
    componentLinearVelScale=<0.5 0.5 0.5>
    physicsBodyDefinitions [{
      collisionType=CustomSphere
      sphereCollisionRadius=2
      boundBone {boneName=dyn_stache01_r}
      boxExtents=<9 18 22>
      constraintSetup {angularLimitsMax=<0 3 3>, angularLimitsMin=<0 -65 -30>, angularTarget=<0 0 0>}
      localJointOffset=<-4 0 0>
     }]
    sphericalLimits [{limitRadius=55, drivingBone {boneName=spine_04}, sphereLocalOffset=<0 -43 0>}]
   }
   n44 = RigidBody!(
    ComponentPose=@n43
    Alpha
    OverrideWorldGravity:Vector=Get gravity_vector() -> (gravity_vector:Vector)
    ExternalForce=Get skydiving_rigid_force() -> (skydiving_rigid_force:Vector)
   ) -> (Pose) {
    bClampLinearTranslationLimitToRefPose=true
    bOverrideWorldGravity=true
    overridePhysicsAsset=&/M_MED_BunkerMan_Beard
    simulationSpace=BaseBoneSpace
    alphaBoolBlend {blendOption=Linear}
    baseBoneRef {boneName=root}
    overrideWorldGravity {z=-700}
   }
   Output Pose(Result=Blend Poses by bool(
     BlendPose_0=Blend Poses by bool(BlendPose_0=@n5, BlendPose_1=@n4, bActiveValue=@n6) {blendType=Linear}
     BlendPose_1=Component To Local(ComponentPose=@n44)
     bActiveValue=Get rigids_off() -> (rigids_off:bool)
    ) {blendType=Linear}) {layerGroup=None}
   n45 = Get head_up_alpha() -> (head_up_alpha:double)
   n46 = Get head_x_negative() -> (head_x_negative:double)
   n47 = Get head_x_positive() -> (head_x_positive:double)
   n48 = "Constraint - faceAttach (P)"(ComponentPose=Local To Component(LocalPose=Use cached pose 'Start_Cache'()), Alpha=1, ConstraintWeights_0=1) {
    alphaBoolBlend {blendOption=Linear}
    boneToModify {boneName=faceAttach}
    constraintSetup [{targetBone {boneName=head}}]
    constraintWeights [1]
   }
   n49 = "Transform (Modify) Bone - Bone: faceAttach"(ComponentPose=@n48, Alpha=1, Translation="0,0,0", Rotation="50,0,0", Scale="1,1,1") {
    rotationMode=BMM_Additive
    rotationSpace=BCS_BoneSpace
    alphaBoolBlend {blendOption=Linear}
    boneToModify {boneName=faceAttach}
    rotation {pitch=50}
   }
   n50 = Control Rig(Source=Blend Poses by bool(BlendPose_0=Component To Local(ComponentPose=@n49), BlendPose_1=Use cached pose 'Start_Cache'(), bActiveValue=Get rigids_off()) {blendType=Linear}) {
    alphaBoolBlend {blendOption=Linear}
    controlRigAssetReference {blueprintRigClass=&/M_MED_BunkerMan_Beard_CtrlRig.M_MED_BunkerMan_Beard_CtrlRig_C}
   }
   n51 = Flap Node - Bone: dyn_hair_bk_01!(ComponentPose:Component Space Pose Link Structure=Local To Component(LocalPose=@n50), Alpha:float=1) -> (Pose:Component Space Pose Link Structure) {
    is=title FortAnimGraphNode_Flap
    bUseClamp=true
    targetClampMax=100
    targetClampMin=-5
    targetComponent=TranslationY
    alphaBoolBlend {blendOption=Linear}
    sourceDrives [{
      bUseRange=false
      drivingCurve=null
      multiplier=6
      rangeMax=1
      rangeMin=0
      remappedMax=1
      remappedMin=0
      sourceComponent=RotationZ
      useQuaternion=true
      sourceBone {boneName=head}
     }]
    targetBone {boneName=dyn_hair_bk_01}
   }
   n52 = Flap Node - Bone: dyn_hair_bk_l_01!(ComponentPose:Component Space Pose Link Structure=@n51, Alpha:float=1) -> (Pose:Component Space Pose Link Structure) {
    is=title FortAnimGraphNode_Flap
    bUseClamp=true
    targetClampMax=100
    targetClampMin=-5
    targetComponent=TranslationY
    alphaBoolBlend {blendOption=Linear}
    sourceDrives [{
      bUseRange=false
      drivingCurve=null
      multiplier=6
      rangeMax=1
      rangeMin=0
      remappedMax=1
      remappedMin=0
      sourceComponent=RotationZ
      useQuaternion=true
      sourceBone {boneName=head}
     }]
    targetBone {boneName=dyn_hair_bk_l_01}
   }
   n53 = Flap Node - Bone: dyn_hair_bk_r_01!(ComponentPose:Component Space Pose Link Structure=@n52, Alpha:float=1) -> (Pose:Component Space Pose Link Structure) {
    is=title FortAnimGraphNode_Flap
    bUseClamp=true
    targetClampMax=100
    targetClampMin=-5
    targetComponent=TranslationY
    alphaBoolBlend {blendOption=Linear}
    sourceDrives [{
      bUseRange=false
      drivingCurve=null
      multiplier=6
      rangeMax=1
      rangeMin=0
      remappedMax=1
      remappedMin=0
      sourceComponent=RotationZ
      useQuaternion=true
      sourceBone {boneName=head}
     }]
    targetBone {boneName=dyn_hair_bk_r_01}
   }
   n54 = Flap Node - Bone: dyn_hair_bk_01(ComponentPose=@n53, Alpha=1) {
    bUseClamp=true
    targetClampMax=100
    targetClampMin=-5
    targetComponent=TranslationX
    alphaBoolBlend {blendOption=Linear}
    sourceDrives [{
      bUseRange=false
      drivingCurve=null
      multiplier=3
      rangeMax=1
      rangeMin=0
      remappedMax=1
      remappedMin=0
      sourceComponent=RotationZ
      useQuaternion=true
      sourceBone {boneName=head}
     }]
    targetBone {boneName=dyn_hair_bk_01}
   }
   n55 = Flap Node - Bone: dyn_hair_bk_l_01(ComponentPose=@n54, Alpha=1) {
    bUseClamp=true
    targetClampMax=100
    targetClampMin=-5
    targetComponent=TranslationX
    alphaBoolBlend {blendOption=Linear}
    sourceDrives [{
      bUseRange=false
      drivingCurve=null
      multiplier=3
      rangeMax=1
      rangeMin=0
      remappedMax=1
      remappedMin=0
      sourceComponent=RotationZ
      useQuaternion=true
      sourceBone {boneName=head}
     }]
    targetBone {boneName=dyn_hair_bk_l_01}
   }
   n56 = Flap Node - Bone: dyn_hair_bk_r_01(ComponentPose=@n55, Alpha=1) {
    bUseClamp=true
    targetClampMax=100
    targetClampMin=-5
    targetComponent=TranslationX
    alphaBoolBlend {blendOption=Linear}
    sourceDrives [{
      bUseRange=false
      drivingCurve=null
      multiplier=3
      rangeMax=1
      rangeMin=0
      remappedMax=1
      remappedMin=0
      sourceComponent=RotationZ
      useQuaternion=true
      sourceBone {boneName=head}
     }]
    targetBone {boneName=dyn_hair_bk_r_01}
   }
   n57 = Flap Node - Bone: dyn_hair_bk_02!(ComponentPose:Component Space Pose Link Structure=@n56, Alpha:float=1) -> (Pose:Component Space Pose Link Structure) {
    is=title FortAnimGraphNode_Flap
    bUseClamp=true
    targetClampMax=100
    targetClampMin=-5
    targetComponent=TranslationX
    alphaBoolBlend {blendOption=Linear}
    sourceDrives [{
      bUseRange=false
      drivingCurve=null
      multiplier=-3
      rangeMax=1
      rangeMin=0
      remappedMax=1
      remappedMin=0
      sourceComponent=RotationZ
      useQuaternion=true
      sourceBone {boneName=head}
     }]
    targetBone {boneName=dyn_hair_bk_02}
   }
   n58 = Flap Node - Bone: dyn_hair_bk_l_02!(ComponentPose:Component Space Pose Link Structure=@n57, Alpha:float=1) -> (Pose:Component Space Pose Link Structure) {
    is=title FortAnimGraphNode_Flap
    bUseClamp=true
    targetClampMax=100
    targetClampMin=-5
    targetComponent=TranslationX
    alphaBoolBlend {blendOption=Linear}
    sourceDrives [{
      bUseRange=false
      drivingCurve=null
      multiplier=-3
      rangeMax=1
      rangeMin=0
      remappedMax=1
      remappedMin=0
      sourceComponent=RotationZ
      useQuaternion=true
      sourceBone {boneName=head}
     }]
    targetBone {boneName=dyn_hair_bk_l_02}
   }
   n59 = Flap Node - Bone: dyn_hair_bk_r_02!(ComponentPose:Component Space Pose Link Structure=@n58, Alpha:float=1) -> (Pose:Component Space Pose Link Structure) {
    is=title FortAnimGraphNode_Flap
    bUseClamp=true
    targetClampMax=100
    targetClampMin=-5
    targetComponent=TranslationX
    alphaBoolBlend {blendOption=Linear}
    sourceDrives [{
      bUseRange=false
      drivingCurve=null
      multiplier=-3
      rangeMax=1
      rangeMin=0
      remappedMax=1
      remappedMin=0
      sourceComponent=RotationZ
      useQuaternion=true
      sourceBone {boneName=head}
     }]
    targetBone {boneName=dyn_hair_bk_r_02}
   }
   n60 = Flap Node - Bone: dyn_hair_bk_02(ComponentPose=@n59, Alpha=1) {
    bUseClamp=true
    targetClampMax=100
    targetClampMin=-5
    targetComponent=TranslationX
    alphaBoolBlend {blendOption=Linear}
    sourceDrives [{
      bUseRange=false
      drivingCurve=null
      multiplier=-6
      rangeMax=1
      rangeMin=0
      remappedMax=1
      remappedMin=0
      sourceComponent=RotationZ
      useQuaternion=true
      sourceBone {boneName=head}
     }]
    targetBone {boneName=dyn_hair_bk_02}
   }
   n61 = Flap Node - Bone: dyn_hair_bk_l_02(ComponentPose=@n60, Alpha=1) {
    bUseClamp=true
    targetClampMax=100
    targetClampMin=-5
    targetComponent=TranslationX
    alphaBoolBlend {blendOption=Linear}
    sourceDrives [{
      bUseRange=false
      drivingCurve=null
      multiplier=-6
      rangeMax=1
      rangeMin=0
      remappedMax=1
      remappedMin=0
      sourceComponent=RotationZ
      useQuaternion=true
      sourceBone {boneName=head}
     }]
    targetBone {boneName=dyn_hair_bk_l_02}
   }
   n62 = Flap Node - Bone: dyn_hair_bk_r_02(ComponentPose=@n61, Alpha=1) {
    bUseClamp=true
    targetClampMax=100
    targetClampMin=-5
    targetComponent=TranslationX
    alphaBoolBlend {blendOption=Linear}
    sourceDrives [{
      bUseRange=false
      drivingCurve=null
      multiplier=-6
      rangeMax=1
      rangeMin=0
      remappedMax=1
      remappedMin=0
      sourceComponent=RotationZ
      useQuaternion=true
      sourceBone {boneName=head}
     }]
    targetBone {boneName=dyn_hair_bk_r_02}
   }
   n63 = "Transform (Modify) Bone - Bone: parent_dyn_beard_02"(ComponentPose=@n62, Alpha=@n47, Translation="0,0,0", Rotation="10,0.000,-30", Scale="1,1,1") {
    rotationMode=BMM_Additive
    rotationSpace=BCS_BoneSpace
    translationMode=BMM_Additive
    translationSpace=BCS_BoneSpace
    alphaBoolBlend {blendOption=Linear}
    boneToModify {boneName=parent_dyn_beard_02}
    rotation {pitch=10, roll=-30}
   }
   n64 = "Transform (Modify) Bone - Bone: dyn_beard_02"(ComponentPose=@n63, Alpha=@n47, Translation="0,15,0", Rotation="0,0,0", Scale="1,1,1") {
    rotationMode=BMM_Additive
    rotationSpace=BCS_BoneSpace
    translationMode=BMM_Additive
    translationSpace=BCS_BoneSpace
    alphaBoolBlend {blendOption=Linear}
    boneToModify {boneName=dyn_beard_02}
    translation {y=15}
   }
   n65 = "Transform (Modify) Bone - Bone: dyn_beard_03"(ComponentPose=@n64, Alpha=@n47, Translation="0,12,0", Rotation="0,0,0", Scale="1,1,1") {
    rotationMode=BMM_Additive
    rotationSpace=BCS_BoneSpace
    translationMode=BMM_Additive
    translationSpace=BCS_BoneSpace
    alphaBoolBlend {blendOption=Linear}
    boneToModify {boneName=dyn_beard_03}
    translation {y=12}
   }
   n66 = "Transform (Modify) Bone - Bone: dyn_beard_04"(ComponentPose=@n65, Alpha=@n47, Translation="0,10,0", Rotation="0,0,0", Scale="1,1,1") {
    rotationMode=BMM_Additive
    rotationSpace=BCS_BoneSpace
    translationMode=BMM_Additive
    translationSpace=BCS_BoneSpace
    alphaBoolBlend {blendOption=Linear}
    boneToModify {boneName=dyn_beard_04}
    translation {y=10}
   }
   n67 = "Transform (Modify) Bone - Bone: dyn_beard_05"(ComponentPose=@n66, Alpha=@n47, Translation="0,10,0", Rotation="0,0,0", Scale="1,1,1") {
    rotationMode=BMM_Additive
    rotationSpace=BCS_BoneSpace
    translationMode=BMM_Additive
    translationSpace=BCS_BoneSpace
    alphaBoolBlend {blendOption=Linear}
    boneToModify {boneName=dyn_beard_05}
    translation {y=10}
   }
   n68 = "Transform (Modify) Bone - Bone: dyn_beard_06"(ComponentPose=@n67, Alpha=1, Translation="0,7,0", Rotation="0,0,0", Scale="1,1,1") {
    rotationMode=BMM_Additive
    rotationSpace=BCS_BoneSpace
    translationMode=BMM_Additive
    translationSpace=BCS_BoneSpace
    alphaBoolBlend {blendOption=Linear}
    boneToModify {boneName=dyn_beard_06}
    translation {y=7}
   }
   n69 = "Transform (Modify) Bone - Bone: dyn_beard_01"(ComponentPose=@n68, Alpha=@n46, Translation="0,0,0", Rotation="5,0,30", Scale="1,1,1") {
    rotationMode=BMM_Additive
    rotationSpace=BCS_BoneSpace
    translationMode=BMM_Additive
    translationSpace=BCS_BoneSpace
    alphaBoolBlend {blendOption=Linear}
    boneToModify {boneName=dyn_beard_01}
    rotation {pitch=5, roll=30}
   }
   n70 = "Transform (Modify) Bone - Bone: parent_dyn_beard_02"(ComponentPose=@n69, Alpha=@n46, Translation="0,0,0", Rotation="10,0,20", Scale="1,1,1") {
    rotationMode=BMM_Additive
    rotationSpace=BCS_BoneSpace
    translationMode=BMM_Additive
    translationSpace=BCS_BoneSpace
    alphaBoolBlend {blendOption=Linear}
    boneToModify {boneName=parent_dyn_beard_02}
    rotation {pitch=10, roll=20}
   }
   n71 = "Transform (Modify) Bone - Bone: dyn_beard_02"(ComponentPose=@n70, Alpha=@n46, Translation="0,-15,0", Rotation="0,0,0", Scale="1,1,1") {
    rotationMode=BMM_Additive
    rotationSpace=BCS_BoneSpace
    translationMode=BMM_Additive
    translationSpace=BCS_BoneSpace
    alphaBoolBlend {blendOption=Linear}
    boneToModify {boneName=dyn_beard_02}
    translation {y=-15}
   }
   n72 = "Transform (Modify) Bone - Bone: dyn_beard_03"(ComponentPose=@n71, Alpha=@n46, Translation="0,-12,0", Rotation="0,0,0", Scale="1,1,1") {
    rotationMode=BMM_Additive
    rotationSpace=BCS_BoneSpace
    translationMode=BMM_Additive
    translationSpace=BCS_BoneSpace
    alphaBoolBlend {blendOption=Linear}
    boneToModify {boneName=dyn_beard_03}
    translation {y=-12}
   }
   n73 = "Transform (Modify) Bone - Bone: dyn_beard_04"(ComponentPose=@n72, Alpha=@n46, Translation="0,-10,0", Rotation="0,0,0", Scale="1,1,1") {
    rotationMode=BMM_Additive
    rotationSpace=BCS_BoneSpace
    translationMode=BMM_Additive
    translationSpace=BCS_BoneSpace
    alphaBoolBlend {blendOption=Linear}
    boneToModify {boneName=dyn_beard_04}
    translation {y=-10}
   }
   n74 = "Transform (Modify) Bone - Bone: dyn_beard_05"(ComponentPose=@n73, Alpha=@n46, Translation="0,-10,0", Rotation="0,0,0", Scale="1,1,1") {
    rotationMode=BMM_Additive
    rotationSpace=BCS_BoneSpace
    translationMode=BMM_Additive
    translationSpace=BCS_BoneSpace
    alphaBoolBlend {blendOption=Linear}
    boneToModify {boneName=dyn_beard_05}
    translation {y=-10}
   }
   n75 = "Transform (Modify) Bone - Bone: dyn_beard_06"(ComponentPose=@n74, Alpha=1, Translation="0,-7,0", Rotation="0,0,0", Scale="1,1,1") {
    rotationMode=BMM_Additive
    rotationSpace=BCS_BoneSpace
    translationMode=BMM_Additive
    translationSpace=BCS_BoneSpace
    alphaBoolBlend {blendOption=Linear}
    boneToModify {boneName=dyn_beard_06}
    translation {y=-7}
   }
   n76 = "Transform (Modify) Bone - Bone: dyn_beard_02"(ComponentPose=@n75, Alpha=@n45, Translation="-20,0,0", Rotation="-20,0,0", Scale="1,1,1") {
    rotationMode=BMM_Additive
    rotationSpace=BCS_BoneSpace
    translationMode=BMM_Additive
    translationSpace=BCS_BoneSpace
    alphaBoolBlend {blendOption=Linear}
    boneToModify {boneName=dyn_beard_02}
    rotation {pitch=-20}
    translation {x=-20}
   }
   n77 = Get head_down_alpha() -> (head_down_alpha:double)
   n78 = "Transform (Modify) Bone - Bone: dyn_beard_02"(ComponentPose=@n76, Alpha=@n77, Translation="5,0,0", Rotation="20,0,0", Scale="1,1,1") {
    rotationMode=BMM_Additive
    rotationSpace=BCS_BoneSpace
    translationMode=BMM_Additive
    translationSpace=BCS_BoneSpace
    alphaBoolBlend {blendOption=Linear}
    boneToModify {boneName=dyn_beard_02}
    rotation {pitch=20}
    translation {x=5}
   }
   n79 = "Transform (Modify) Bone - Bone: dyn_beard_03"(ComponentPose=@n78, Alpha=@n45, Translation="-5,0,0", Rotation="-10,0,0", Scale="1,1,1") {
    rotationMode=BMM_Additive
    rotationSpace=BCS_BoneSpace
    translationMode=BMM_Additive
    translationSpace=BCS_BoneSpace
    alphaBoolBlend {blendOption=Linear}
    boneToModify {boneName=dyn_beard_03}
    rotation {pitch=-10}
    translation {x=-5}
   }
   n80 = "Transform (Modify) Bone - Bone: dyn_beard_03"(ComponentPose=@n79, Alpha=@n77, Translation="5,0,0", Rotation="30,0,0", Scale="1,1,1") {
    rotationMode=BMM_Additive
    rotationSpace=BCS_BoneSpace
    translationMode=BMM_Additive
    translationSpace=BCS_BoneSpace
    alphaBoolBlend {blendOption=Linear}
    boneToModify {boneName=dyn_beard_03}
    rotation {pitch=30}
    translation {x=5}
   }
   n81 = "Transform (Modify) Bone - Bone: dyn_beard_04"(ComponentPose=@n80, Alpha=@n45, Translation="-5,0,0", Rotation="5,0,0", Scale="1,1,1") {
    rotationMode=BMM_Additive
    rotationSpace=BCS_BoneSpace
    translationMode=BMM_Additive
    translationSpace=BCS_BoneSpace
    alphaBoolBlend {blendOption=Linear}
    boneToModify {boneName=dyn_beard_04}
    rotation {pitch=5}
    translation {x=-5}
   }
   n82 = "Transform (Modify) Bone - Bone: dyn_beard_04"(ComponentPose=@n81, Alpha=@n77, Translation="5,0,0", Rotation="30,0,0", Scale="1,1,1") {
    rotationMode=BMM_Additive
    rotationSpace=BCS_BoneSpace
    translationMode=BMM_Additive
    translationSpace=BCS_BoneSpace
    alphaBoolBlend {blendOption=Linear}
    boneToModify {boneName=dyn_beard_04}
    rotation {pitch=30}
    translation {x=5}
   }
   n83 = "Transform (Modify) Bone - Bone: dyn_beard_05"(ComponentPose=@n82, Alpha=@n45, Translation="-2.5,0,0", Rotation="0,0,0", Scale="1,1,1") {
    rotationMode=BMM_Additive
    rotationSpace=BCS_BoneSpace
    translationMode=BMM_Additive
    translationSpace=BCS_BoneSpace
    alphaBoolBlend {blendOption=Linear}
    boneToModify {boneName=dyn_beard_05}
    translation {x=-2.5}
   }
   n84 = "Transform (Modify) Bone - Bone: dyn_beard_05"(ComponentPose=@n83, Alpha=@n77, Translation="2.5,0,0", Rotation="20,0,0", Scale="1,1,1") {
    rotationMode=BMM_Additive
    rotationSpace=BCS_BoneSpace
    translationMode=BMM_Additive
    translationSpace=BCS_BoneSpace
    alphaBoolBlend {blendOption=Linear}
    boneToModify {boneName=dyn_beard_05}
    rotation {pitch=20}
    translation {x=2.5}
   }
   n85 = "Constraint - dyn_beard_02 (P)"(ComponentPose=@n84, Alpha=0.65, ConstraintWeights_0=1) {
    alphaBoolBlend {blendOption=Linear}
    boneToModify {boneName=dyn_beard_02}
    constraintSetup [{targetBone {boneName=spine_04}}]
    constraintWeights [1]
   }
   n86 = "Constraint - dyn_beard_03 (P)"(ComponentPose=@n85, Alpha=0.6, ConstraintWeights_0=1) {
    alphaBoolBlend {blendOption=Linear}
    boneToModify {boneName=dyn_beard_03}
    constraintSetup [{targetBone {boneName=spine_04}}]
    constraintWeights [1]
   }
   n87 = "Constraint - dyn_beard_04 (P)"(ComponentPose=@n86, Alpha=0.4, ConstraintWeights_0=1) {
    alphaBoolBlend {blendOption=Linear}
    boneToModify {boneName=dyn_beard_04}
    constraintSetup [{targetBone {boneName=spine_04}}]
    constraintWeights [1]
   }
   n88 = "Constraint - dyn_beard_05 (P)"(ComponentPose=@n87, Alpha=0.2, ConstraintWeights_0=1) {
    alphaBoolBlend {blendOption=Linear}
    boneToModify {boneName=dyn_beard_05}
    constraintSetup [{targetBone {boneName=spine_03}}]
    constraintWeights [1]
   }
   n89 = "Constraint - dyn_beard_06 (P)"(ComponentPose=@n88, Alpha=0.15, ConstraintWeights_0=1) {
    alphaBoolBlend {blendOption=Linear}
    boneToModify {boneName=dyn_beard_06}
    constraintSetup [{targetBone {boneName=spine_01}}]
    constraintWeights [1]
   }
   Save cached pose 'AnimCache'(Pose=Component To Local(ComponentPose=@n89)) {cacheName=AnimCache}
   n90 = Layered blend per bone(
    BasePose=Copy Pose From Mesh() {bCopyCurves=true, bUseAttachedParent=true}
    BlendPoses_0=Copy Pose From Mesh(SourceMeshComponent=Get Body_Mesh())
   ) {layerSetup [{branchFilters [{boneName=thigh_l}, {boneName=thigh_r}, {boneName=calf_l}, {boneName=calf_r}]}]}
   Save cached pose 'Start_Cache'(Pose=@n90) {cacheName=Start_Cache}
   n91 = Flap Node - Bone: dyn_hair_bk_01(Alpha=1) {
    bUseClamp=true
    targetClampMax=100
    targetComponent=RotationZ
    alphaBoolBlend {blendOption=Linear}
    sourceDrives [{
      bUseRange=false
      drivingCurve=null
      multiplier=3
      rangeMax=1
      rangeMin=0
      remappedMax=1
      remappedMin=0
      sourceComponent=RotationZ
      useQuaternion=true
      sourceBone {boneName=head}
     }]
    targetBone {boneName=dyn_hair_bk_01}
   }
   n92 = Flap Node - Bone: dyn_hair_bk_l_01(ComponentPose=@n91, Alpha=1) {
    bUseClamp=true
    targetClampMax=100
    targetComponent=RotationZ
    alphaBoolBlend {blendOption=Linear}
    sourceDrives [{
      bUseRange=false
      drivingCurve=null
      multiplier=3
      rangeMax=1
      rangeMin=0
      remappedMax=1
      remappedMin=0
      sourceComponent=RotationZ
      useQuaternion=true
      sourceBone {boneName=head}
     }]
    targetBone {boneName=dyn_hair_bk_l_01}
   }
   Flap Node - Bone: dyn_hair_bk_r_03!(ComponentPose:Component Space Pose Link Structure=@n92, Alpha:float=1) -> (Pose:Component Space Pose Link Structure) {
    is=title FortAnimGraphNode_Flap
    bUseClamp=true
    targetClampMax=100
    targetComponent=RotationZ
    alphaBoolBlend {blendOption=Linear}
    sourceDrives [{
      bUseRange=false
      drivingCurve=null
      multiplier=3
      rangeMax=1
      rangeMin=0
      remappedMax=1
      remappedMin=0
      sourceComponent=RotationZ
      useQuaternion=true
      sourceBone {boneName=head}
     }]
    targetBone {boneName=dyn_hair_bk_r_03}
   }
  ]}
 EventGraph {nodes [
   Event BlueprintInitializeAnimation()
   > Set Body_Mesh(Body_Mesh=GetOwnerSkeletalMeshComponent())
   > n1 = Cast To PlayerPawn_Athena(Object=TryGetPawnOwner())
   > n2 = Set Pawn Owner(Pawn Owner:Player Pawn Athena Object Reference=@n1) -> (Output_Get:Player Pawn Athena Object Reference)
   > Set Head_Mesh(Head_Mesh:Skeletal Mesh Component Object Reference=GetSkeletalMeshForPartType(self=@n2, PartType=Head)) -> (Output_Get:Skeletal Mesh Component Object Reference)
   Event BlueprintUpdateAnimation()
   > Is Valid(InputObject=Get Head_Mesh() -> (Head_Mesh:Skeletal Mesh Component Object Reference))
   > n3 = Sequence!(execute) -> (then_0, then_1, then_2:Exec, then_3:Exec, then_4:Exec)
   n4 = Get PawnSpeed2D()
   > n5 = Branch(Condition="float <= float"!(A:double=MapRangeClamped(Value=@n4, InRangeA=0, InRangeB=100), B:double=0) -> (ReturnValue:bool) {is=title K2Node_CallFunction})
   > Set speed_mult(speed_mult:double=0.1) -> (Output_Get:double)
   > Set hair_alpha(hair_alpha:double=0.35) -> (Output_Get:double)
   Set hair_alpha(execute=@n5.else, hair_alpha=0.65)
   n6 = Branch(execute=@n3.then_3, Condition=Get bIsSkyDiving())
   n7 = Break Fort Anim Input Skydiving External Force(FortAnimInput_SkydivingExternalForce=Get SkydivingExternalForceInput()) -> (
    bUseSkydivingVectorForce=~
    SkydivingVectorForce=~
    HeadToPelvisDirection=~
    FloatingMultiplier=~
    FloatingAdditive=~
    DivingMultiplier=~
    DivingAdditive=~
    ParachutingMultiplier=~
    ParachutingAdditive=~
   )
   > Set skydiving_force(skydiving_force:Vector=vector * float(A=@n7.SkydivingVectorForce, B=5)) -> (Output_Get:Vector)
   > Set skydiving_rigid_force(skydiving_rigid_force:Vector=vector * float(A=@n7.SkydivingVectorForce, B=55)) -> (Output_Get:Vector)
   > n8 = Set gravity_vector(gravity_vector:Vector="0, 0,-980") -> (Output_Get:Vector)
   Branch(execute=[@n8.then, @n9.then], Condition=Get bIsSkydiveFloating())
   Set skydiving_force(execute=@n6.else, skydiving_force="0, 0, 0")
   > Set skydiving_rigid_force(skydiving_rigid_force="0, 0, 0")
   > n9 = Set gravity_vector(gravity_vector="0, 0,-700")
   n10 = Get CurrentLOD()
   n11 = Get bRigidBodyEnabled()
   Set rigids_off(execute=@n3.then_4, rigids_off:bool=OR Boolean(A=NOT Boolean(A=@n11), B="integer > integer"(A=@n10, B=1) {is=title K2Node_CallFunction})) -> (Output_Get:bool)
   n12 = GetOwnerSkeletalMeshComponent()
   n13 = BreakRotator(InRot=BreakTransform(InTransform=GetDeltaTransformFromRefPose(self=@n12, BoneName=spine_01, BaseName=head)).Rotation)
   Set head_down_alpha(
    execute=@n3.then_1
    head_down_alpha:double=float - float(B=MapRangeClamped(Value=@n13.Yaw, InRangeA=-70, InRangeB=70)) {is=title K2Node_CallFunction}
   ) -> (Output_Get:double)
   > Set head_up_alpha(head_up_alpha:double=MapRangeClamped(Value=@n13.Yaw, InRangeA=-70, InRangeB=70)) -> (Output_Get:double)
   Set head_x_negative(execute=@n3.then_2, head_x_negative:double=float - float(B=MapRangeClamped(Value=@n13, InRangeA=-40, InRangeB=0))) -> (Output_Get:double)
   > Set head_x_positive(head_x_positive:double=MapRangeClamped(Value=@n13, InRangeA=0, InRangeB=40)) -> (Output_Get:double)
   n14 = BreakTransform(InTransform=GetDeltaTransformFromRefPose(self=Get Head_Mesh(), BoneName=C_jaw, BaseName=head))
   > Set jaw_open(jaw_open:double=MapRangeClamped(Value=BreakRotator(InRot=@n14.Rotation), InRangeA=0, InRangeB=35)) -> (Output_Get:double)
  ]}
]
variables [
 Body_Mesh {type=SkeletalMeshComponent}
 gravity_vector {type=Vector}
 hair_alpha {type=number, default=1}
 head_down_alpha {type=number}
 Head_Mesh {type=SkeletalMeshComponent}
 head_up_alpha {type=number}
 head_x_negative {type=number}
 head_x_positive {type=number}
 jaw_open {type=number}
 Pawn Owner {type=PlayerPawn_Athena_C}
 rigids_off {type=boolean}
 skydiving_force {type=Vector}
 skydiving_rigid_force {type=Vector}
 speed_mult {type=number}
]
